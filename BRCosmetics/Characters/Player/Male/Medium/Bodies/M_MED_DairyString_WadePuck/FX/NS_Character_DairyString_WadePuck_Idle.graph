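# Idle Niagara effect for the DairyString WadePuck character body that shows green capsule glow sprites on three sockets.
# One emitter bursts three persistent sprites, and each uses its spawn group index to follow a socket through the system Socket Reader.
# Sprites align to socket rotation with slight position and camera offsets.
# Alpha switches between a constant in gameplay and an age-based fade-in in the front end, scaled by a time-of-day value.
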NiagaraSystem /BRCosmetics/Characters/Player/Male/Medium/Bodies/M_MED_DairyString_WadePuck/FX/NS_Character_DairyString_WadePuck_Idle
dependencies {
 usedDataInterfaces [&NiagaraDataInterfaceCurve, &NiagaraDataInterfaceSocketReader]
 usedDynamicInputs [&FNTimeOfDayScalar]
 usedModules [
  &self:SystemSpawnScript.SystemScriptSource.SystemScriptGraph.NiagaraNodeAssignment_0.SetVariables_#1
  &SystemState
  &self:SystemSpawnScript.SystemScriptSource.SystemScriptGraph.NiagaraNodeAssignment_2.SetVariables_#2
 ]
}
emitters {CapsuleGlow {
  emitterUpdateScript [
   &EmitterState(
    Inactive Response=::NewEnumerator1
    Life Cycle Mode=::NewEnumerator1
    Loop Behavior=::NewEnumerator1
    Loop Duration Mode=::NewEnumerator1
    UseLoopDelay=false
   )
   &SpawnBurst_Instantaneous(Spawn Count=3, Use Spawn Probability=false, UseLoopCountLimit=false)
  ]
  particleSpawnScript [
   &FN_SetSpawnGroup(Number of Spawn Groups=3)
   &InitializeParticle(
    Sprite Rotation Mode=::NewEnumerator0
    Sprite Size=<15 28>
    Sprite Size Mode=::NewEnumerator4
    UsePositionOffset=true
    Write Material Random=true
   )
   &CameraOffset(Camera Offset Amount=4)
  ]
  particleUpdateScript [
   &ParticleState(Kill Particles When Lifetime Has Elapsed=false, Let Infinitely Lived Particles Die When Emitter Deactivates=true)
   &ScaleColor(
    Scale Alpha=&Multiply_Float(
     A=&Multiply_Float(
      A=&FN_FrontEndFloatModifier(
       Gameplay=0.15
       Front End=&FloatFromCurve(FloatCurve=DataInterface() {propertyValues {Curve {keys [<at 0.2 0>,<at 0.7 1>]}}}, CurveIndex=$System.Age, Scale Curve=0.05)
      )
      B=$System.TODFloat
     )
     B=0.8
    )
    Scale RGB=&MakeVectorFromLinearColorRGB(Color=<rgba 0.6365 3.1 0.1437 1>)
    ScaleA=true
    ScaleRGB=true
   )
   &SocketLocation(
    SamplePosOffset=true
    Sampled Position Offset=&MergeHelper_VectorSwitch(0=<1.5 0 0>, 1=<1.5 0 0>, 2=<1.5 0 0>, 3=?, Index=&SpawnGroup(), Num=::NewEnumerator1)
    Socket Index=&SpawnGroup()
    Socket Reader=$System.Socket Reader
    Socket Sampling Mode=::NewEnumerator3
    VelocitySpeedLimit=false
   )
   &SpriteFacingAndAlignment(
    Sprite Alignment=&RotateVectorByQuaternion(Vector To Rotate=<1 0 0>, Quaternion=$Output.SocketLocation.SampledRotation)
    WriteAlignment=true
    WriteFacing=true
   ) {module=SpriteFacingAndAlignment001}
  ]
  rendererClasses [&NiagaraSpriteRendererProperties]
  renderers [{rendererClass=&NiagaraSpriteRendererProperties}]
 }}
summary {
 systemName=NS_Character_DairyString_WadePuck_Idle
 emitters [{emitterName=CapsuleGlow, rendererClasses [&NiagaraSpriteRendererProperties]}]
 userVariables [
  User.bIsFrontEnd {type=NiagaraBool}
  User.bIsFrontEndPreview {type=NiagaraBool}
 ]
}
