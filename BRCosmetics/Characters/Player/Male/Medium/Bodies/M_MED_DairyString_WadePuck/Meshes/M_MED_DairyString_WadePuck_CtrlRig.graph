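# Control Rig for the WadePuck player body cosmetic, correcting cloth, accessory and arm deformation.
# Pose readers, remapped joint rotations, anim attributes and state variables (knee slide, DBNO, emote) weight additive bone offsets for coat cloth, lapels and sleeves.
# Deform arm bones copy the animated arms, solve PBIK to the hands, twist-correct for gauntlets, and drive twist and elbow helpers.
# Weighted parent constraints attach bow and tube bones to the spine.
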ControlRigBlueprint /BRCosmetics/Characters/Player/Male/Medium/Bodies/M_MED_DairyString_WadePuck/Meshes/M_MED_DairyString_WadePuck_CtrlRig
graphs [
 New Function_ContainedGraph {nodes [
   n1 = RigVMFunctionEntryNode()
   RigVMFunction_MathQuaternionToEuler = &RigVMFunction_MathQuaternionToEuler(
    RotationOrder=XYZ
    Value=&RigUnit_GetRelativeTransformForItem(
     Child=(Type=Bone,Name="hand_l")
     Parent=(Type=Null,Name="lowerarm_l")
     bChildInitial=False
     bParentInitial=False
     Child.Name=@n1.Child
     Parent.Name=@n1.Parent
    ).RelativeTransform.Rotation
   )
   Remap = &RigVMFunction_MathDoubleRemap(
    TargetMaximum=1
    TargetMinimum=0
    Value=@RigVMFunction_MathQuaternionToEuler.Result.Z
    SourceMaximum=@n1.Source Maximum
    SourceMinimum=@n1.Source Minimum
   )
   > RigVMFunctionReturnNode(Delta Rotation=@Remap.Result)
  ]}
 New Function_ContainedGraph {nodes [
   n1 = RigVMFunctionEntryNode()
   RigVMFunction_MathQuaternionToEuler = &RigVMFunction_MathQuaternionToEuler(
    RotationOrder=XYZ
    Value=&RigUnit_GetRelativeTransformForItem(
     Child=(Type=Bone,Name="hand_l")
     Parent=(Type=Null,Name="lowerarm_l")
     bChildInitial=False
     bParentInitial=False
     Child.Name=@n1.Child
     Parent.Name=@n1.Parent
    ).RelativeTransform.Rotation
   )
   Remap = &RigVMFunction_MathDoubleRemap(
    TargetMaximum=1
    TargetMinimum=0
    Value=@RigVMFunction_MathQuaternionToEuler.Result.Y
    SourceMaximum=@n1.Source Maximum
    SourceMinimum=@n1.Source Minimum
   )
   > RigVMFunctionReturnNode(Delta Rotation=@Remap.Result)
  ]}
 New Function_ContainedGraph {nodes [
   RigVMFunctionEntryNode()
   > &RigUnit_ParentConstraint(
    AdvancedSettings=(RotationOrderForFilter=XZY)
    Child=(Type=Bone,Name="base_bow_m")
    Parents=((Item=(Type=Bone,Name="neck_01"),Weight=0.0029),(Item=(Type=Bone,Name="spine_02"),Weight=0.0439),(Item=(Type=Bone,Name="spine_03"),Weight=0.1609),(Item=(Type=Bone,Name="spine_05"),Weight=0.3288),(Item=(Type=Bone,Name="spine_04"),Weight=0.4636))
    Weight=1
   )
   > &RigUnit_ParentConstraint(
    AdvancedSettings=(RotationOrderForFilter=XZY)
    Child=(Type=Bone,Name="base_tube_01_m")
    Parents=((Item=(Type=Bone,Name="spine_01"),Weight=0.0232),(Item=(Type=Bone,Name="spine_05"),Weight=0.1268),(Item=(Type=Bone,Name="spine_02"),Weight=0.1742),(Item=(Type=Bone,Name="spine_03"),Weight=0.2848),(Item=(Type=Bone,Name="spine_04"),Weight=0.3911))
    Weight=1
   )
   > &RigUnit_ParentConstraint(
    AdvancedSettings=(RotationOrderForFilter=XZY)
    Child=(Type=Bone,Name="base_tube_02_m")
    Parents="((Item=(Type=Bone,Name=\"spine_01\"),Weight=0.0338),(Item=(Type=Bone,Name=\"spine_05\"),Weight=0.0643),(Item=(Type=Bone,Name=\"pelvis\"),Weight=0.1003),(Item=(Type=Bone,Name=\"spine_02\"),Weight=0.2472),(Item=(Type=Bone,Name=\"spine_04\"),Weight=0.2648),(Item=(Type= ...(truncated)"
    Weight=1
   )
   > &RigUnit_ParentConstraint(
    AdvancedSettings=(RotationOrderForFilter=XZY)
    Child=(Type=Bone,Name="base_tube_03_m")
    Parents="((Item=(Type=Bone,Name=\"thigh_twist_01_r\"),Weight=0.0021),(Item=(Type=Bone,Name=\"spine_05\"),Weight=0.0284),(Item=(Type=Bone,Name=\"thigh_twist_01_l\"),Weight=0.029),(Item=(Type=Bone,Name=\"spine_01\"),Weight=0.0949),(Item=(Type=Bone,Name=\"spine_02\"),Weight=0.1 ...(truncated)"
    Weight=1
   )
   > &RigUnit_ModifyTransforms(
    ItemToModify="((Item=(Type=Bone,Name=\"base_tube_03_m\"),Transform=(Rotation=(X=0,Y=0,Z=0,W=1),Translation=(X=0,Y=0,Z=0.5),Scale3D=(X=1,Y=1,Z=1))),(Item=(Type=Bone,Name=\"base_tube_02_m\"),Transform=(Rotation=(X=0,Y=0,Z=0,W=1),Translation=(X=0,Y=0,Z=0.5),Scale3D=(X=1,Y=1,Z= ...(truncated)"
    Weight=1
    WeightMinimum=0
   )
   > RigVMFunctionReturnNode()
  ]}
 New Function_ContainedGraph {nodes [
   RigVMFunctionEntryNode()
   > RigVMFunction_Sequence = RigVMAggregateNode()
   &RigUnit_ModifyTransforms(
    ItemToModify="((Item=(Type=Bone,Name=\"base_cloth_r_fr\"),Transform=(Rotation=(X=-0.3038,Y=-0.0647,Z=0.0874,W=0.9465),Translation=(X=0,Y=0,Z=0),Scale3D=(X=1,Y=1,Z=1))),(Item=(Type=Bone,Name=\"base_cloth_r\"),Transform=(Rotation=(X=-0.0381,Y=0.0126,Z=0.0358,W=0.9986),Transla ...(truncated)"
    WeightMinimum=0
    Weight=RigVMDispatchNode(BoneName=root, Default=1, Name=deformRig_r_thigh_up_alpha) {template="DISPATCH_RigDispatch_GetAnimAttribute(in Name,in BoneName,in Default,out Value,out Found)"}.Value
    ExecutePin=@RigVMFunction_Sequence.A
   )
   > Delta Rotation Y = RigVMFunctionReferenceNode(Child=thigh_r, Parent=thigh_r_null, Source Maximum=-85)
   > &RigUnit_ModifyTransforms(
    ItemToModify="((Item=(Type=Bone,Name=\"base_cloth_r_fr\"),Transform=(Rotation=(X=-0.3822,Y=-0.5495,Z=0.4719,W=0.5738),Translation=(X=0,Y=0,Z=0),Scale3D=(X=1,Y=1,Z=1))),(Item=(Type=Bone,Name=\"base_cloth_r\"),Transform=(Rotation=(X=-0.592,Y=-0.5201,Z=-0.0217,W=0.6153),Transl ...(truncated)"
    WeightMinimum=0
    Weight=@"Delta Rotation Y".Delta Rotation
   )
   > &RigUnit_ModifyTransforms(
    ItemToModify="((Item=(Type=Bone,Name=\"base_cloth_l_fr\"),Transform=(Rotation=(X=-0.3385,Y=-0.0053,Z=-0.2399,W=-0.9099),Translation=(X=-6.493,Y=-5.8357,Z=0.4588),Scale3D=(X=1,Y=1,Z=1))),(Item=(Type=Bone,Name=\"base_cloth_l\"),Transform=(Rotation=(X=-0.0783,Y=-0.1325,Z=-0.08 ...(truncated)"
    WeightMinimum=0
    Weight=RigVMDispatchNode(BoneName=root, Default=1, Name=deformRig_l_thigh_up_alpha) {template="DISPATCH_RigDispatch_GetAnimAttribute(in Name,in BoneName,in Default,out Value,out Found)"}.Value
   )
   > Delta Rotation Y_1 = RigVMFunctionReferenceNode(Child=thigh_l, Parent=thigh_l_null, Source Maximum=-89, Source Minimum=-10)
   > &RigUnit_ModifyTransforms(
    ItemToModify="((Item=(Type=Bone,Name=\"base_cloth_l_fr\"),Transform=(Rotation=(X=0.3586,Y=-0.5769,Z=-0.49,W=0.5462),Translation=(X=0,Y=0,Z=0),Scale3D=(X=1,Y=1,Z=1))),(Item=(Type=Bone,Name=\"base_cloth_l\"),Transform=(Rotation=(X=0.592,Y=-0.5201,Z=0.0217,W=0.6153),Translatio ...(truncated)"
    WeightMinimum=0
    Weight=@"Delta Rotation Y_1".Delta Rotation
   )
   > &RigUnit_ModifyTransforms(
    ItemToModify="((Item=(Type=Bone,Name=\"base_cloth_l_bck\"),Transform=(Rotation=(X=-0.0819,Y=-0.3096,Z=-0.1764,W=0.9308),Translation=(X=0,Y=0,Z=0),Scale3D=(X=1,Y=1,Z=1))),(Item=(Type=Bone,Name=\"base_cloth_bck\"),Transform=(Rotation=(X=0.1452,Y=-0.2906,Z=0.0623,W=0.9437),Tra ...(truncated)"
    WeightMinimum=0
    Weight=RigVMVariableNode() {variable=thigh_L_Back}.Value
   )
   > &RigUnit_ModifyTransforms(
    ItemToModify="((Item=(Type=Bone,Name=\"base_cloth_r_bck\"),Transform=(Rotation=(X=0.0819,Y=-0.3096,Z=0.1764,W=0.9308),Translation=(X=0,Y=0,Z=0),Scale3D=(X=1,Y=1,Z=1))),(Item=(Type=Bone,Name=\"base_cloth_bck\"),Transform=(Rotation=(X=-0.1452,Y=-0.2906,Z=-0.0623,W=0.9437),Tra ...(truncated)"
    WeightMinimum=0
    Weight=RigVMVariableNode() {variable=thigh_R_Back}.Value
   )
   RigVMFunction_MathBoolToFloat = &RigVMFunction_MathBoolToFloat(Value=RigVMVariableNode() {variable=isKneeSlide}.Value)
   AlphaInterp = &RigVMFunction_AlphaInterp(InterpSpeedDecreasing=10, InterpSpeedIncreasing=10, Scale=1, bClampResult=true, Value=@RigVMFunction_MathBoolToFloat.Result)
   RigVMFunction_AnimEvalRichCurve = &RigVMFunction_AnimEvalRichCurve(Curve=(EditorCurveData=(Keys=((),(Time=0.5503,Value=0.0275),(Time=1,Value=1)))), SourceMinimum=0, TargetMinimum=0, Value=@AlphaInterp.Result)
   > n1 = &RigUnit_ModifyTransforms(
    ItemToModify="((Item=(Type=Bone,Name=\"base_cloth_r_bck\"),Transform=(Rotation=(X=0,Y=0.78,Z=0,W=-0.6257),Translation=(X=0,Y=0,Z=0),Scale3D=(X=1,Y=1,Z=1))),(Item=(Type=Bone,Name=\"base_cloth_bck\"),Transform=(Rotation=(X=0,Y=0.9058,Z=0,W=-0.4237),Translation=(X=0,Y=0,Z=0),S ...(truncated)"
    WeightMinimum=0
    Weight=@RigVMFunction_AnimEvalRichCurve.Result
   )
   RigVMFunction_ControlFlowBranch = &RigVMFunction_ControlFlowBranch(ExecuteContext=@n1.ExecutePin, Condition=RigVMVariableNode() {variable=isEmoting}.Value)
   n2 = &RigVMFunction_AlphaInterp(
    InterpSpeedDecreasing=10
    InterpSpeedIncreasing=10
    Scale=1
    bClampResult=False
    bInterpResult=False
    Value=&RigVMFunction_MathBoolToFloat(Value=RigVMVariableNode() {variable=isDBNO}.Value).Result
   )
   n3 = &RigVMFunction_AnimEvalRichCurve(
    Curve="(EditorCurveData=(Keys=((InterpMode=RCIM_Linear,TangentMode=RCTM_Auto,TangentWeightMode=RCTWM_WeightedNone,Time=0,Value=0,ArriveTangent=0,ArriveTangentWeight=0,LeaveTangent=0,LeaveTangentWeight=0),(InterpMode=RCIM_Linear,TangentMode=RCTM_Auto,TangentWeight ...(truncated)"
    SourceMinimum=0
    TargetMinimum=0
    Value=@n2.Result
   )
   &RigUnit_ModifyTransforms(
    ItemToModify="((Item=(Type=Bone,Name=\"base_cloth_r_bck\"),Transform=(Rotation=(X=0,Y=0,Z=0,W=1),Translation=(X=0,Y=0,Z=0),Scale3D=(X=1,Y=1,Z=1))),(Item=(Type=Bone,Name=\"base_cloth_bck\"),Transform=(Rotation=(X=0,Y=-0.1944,Z=0,W=0.9809),Translation=(X=0,Y=0,Z=0),Scale3D=(X ...(truncated)"
    WeightMinimum=0
    Weight=@n3.Result
    ExecutePin=@RigVMFunction_ControlFlowBranch.Completed
   )
   n4 = &RigUnit_SphericalPoseReader(
    ActiveRegionScaleFactors=(PositiveWidth=1,NegativeWidth=1,PositiveHeight=1,NegativeHeight=1)
    ActiveRegionSize=0.05
    Debug=(bDrawDebug=false,bDraw2D=false,bDrawLocalAxes=False,DebugScale=25,DebugSegments=20,DebugThickness=0.25)
    DriverAxis=(X=1,Y=0,Z=0)
    DriverItem=(Type=Bone,Name="thigh_r")
    FalloffRegionScaleFactors=(PositiveWidth=1,NegativeWidth=1,PositiveHeight=1,NegativeHeight=1)
    FalloffSize=0.55
    FlipHeightScaling=true
    FlipWidthScaling=true
    OptionalParentItem=(Type=Bone,Name="None")
    RotationOffset=(X=-40,Y=20,Z=0)
    ExecutePin=@RigVMFunction_ControlFlowBranch.True
   )
   > &RigUnit_ModifyTransforms(
    ItemToModify="((Item=(Type=Bone,Name=\"base_cloth_r_fr\"),Transform=(Rotation=(X=-0.2343,Y=-0.0728,Z=0.0215,W=0.9692),Translation=(X=-4.7128,Y=1.8208,Z=4.4062),Scale3D=(X=1,Y=1,Z=1))),(Item=(Type=Bone,Name=\"base_cloth_r\"),Transform=(Rotation=(X=-0.0376,Y=-0.1056,Z=0.078,W ...(truncated)"
    WeightMinimum=0
    Weight=@n4.OutputParam
   )
   > n5 = &RigUnit_SphericalPoseReader(
    ActiveRegionScaleFactors=(PositiveWidth=1,NegativeWidth=1,PositiveHeight=1,NegativeHeight=1)
    ActiveRegionSize=0.05
    Debug=(bDrawDebug=false,bDraw2D=False,bDrawLocalAxes=False,DebugScale=25,DebugSegments=20,DebugThickness=0.25)
    DriverAxis=(X=-1,Y=0,Z=0)
    DriverItem=(Type=Bone,Name="thigh_l")
    FalloffRegionScaleFactors=(PositiveWidth=1,NegativeWidth=0.35,PositiveHeight=1,NegativeHeight=1)
    FalloffSize=0.55
    FlipHeightScaling=False
    FlipWidthScaling=False
    OptionalParentItem=(Type=Bone,Name="None")
    RotationOffset=(X=140,Y=30,Z=0)
   )
   > &RigUnit_ModifyTransforms(
    ItemToModify="((Item=(Type=Bone,Name=\"base_cloth_l_fr\"),Transform=(Rotation=(X=0.0438,Y=0.1873,Z=-0.2233,W=-0.9556),Translation=(X=-1.638,Y=-6.3122,Z=4.0171),Scale3D=(X=1,Y=1,Z=1))),(Item=(Type=Bone,Name=\"base_cloth_l\"),Transform=(Rotation=(X=0.0149,Y=0.0378,Z=-0.1315,W ...(truncated)"
    WeightMinimum=0
    Weight=@n5.OutputParam
   )
   n6 = &RigUnit_SphericalPoseReader(
    ActiveRegionScaleFactors=(PositiveWidth=1,NegativeWidth=1,PositiveHeight=1,NegativeHeight=1)
    ActiveRegionSize=0.05
    Debug=(bDrawDebug=false,bDraw2D=false,bDrawLocalAxes=False,DebugScale=25,DebugSegments=20,DebugThickness=0.25)
    DriverAxis=(X=1,Y=0,Z=0)
    DriverItem=(Type=Bone,Name="upperarm_l")
    FalloffRegionScaleFactors=(PositiveWidth=1,NegativeWidth=1,PositiveHeight=1,NegativeHeight=1)
    FalloffSize=0.3
    FlipHeightScaling=true
    FlipWidthScaling=true
    OptionalParentItem=(Type=Bone,Name="spine_05")
    RotationOffset=(X=104.1658,Y=-28.751,Z=-8.8943)
    ExecutePin=@RigVMFunction_Sequence.B
   )
   > &RigUnit_ModifyTransforms(
    ItemToModify=((Item=(Type=Bone,Name="dyn_coatLapel_l"),Transform=(Rotation=(X=0.7811,Y=0,Z=0,W=0.6244),Translation=(X=0,Y=0,Z=0),Scale3D=(X=1,Y=1,Z=1))))
    WeightMinimum=0
    Weight=@n6.OutputParam
   )
   > n7 = &RigUnit_SphericalPoseReader(
    ActiveRegionScaleFactors=(PositiveWidth=1,NegativeWidth=1,PositiveHeight=1,NegativeHeight=1)
    ActiveRegionSize=0.05
    Debug=(bDrawDebug=false,bDraw2D=false,bDrawLocalAxes=False,DebugScale=25,DebugSegments=20,DebugThickness=0.25)
    DriverAxis=(X=-1,Y=0,Z=0)
    DriverItem=(Type=Bone,Name="upperarm_r")
    FalloffRegionScaleFactors=(PositiveWidth=1,NegativeWidth=1,PositiveHeight=1,NegativeHeight=1)
    FalloffSize=0.3
    FlipHeightScaling=true
    FlipWidthScaling=true
    OptionalParentItem=(Type=Bone,Name="spine_05")
    RotationOffset=(X=-93.0515,Y=-49.9553,Z=14.949)
   )
   > &RigUnit_ModifyTransforms(
    ItemToModify=((Item=(Type=Bone,Name="dyn_coatLapel_r"),Transform=(Rotation=(X=0.7811,Y=0,Z=0,W=0.6244),Translation=(X=0,Y=0,Z=0),Scale3D=(X=1,Y=1,Z=1))))
    WeightMinimum=0
    Weight=@n7.OutputParam
   )
   Delta Rotation Z = RigVMFunctionReferenceNode(Child=thigh_l, Parent=thigh_l_null, Source Maximum=10, ExecuteContext=@RigVMFunction_Sequence.C)
   n8 = &RigVMFunction_AlphaInterp(
    InterpSpeedDecreasing=10
    InterpSpeedIncreasing=10
    Scale=1
    bClampResult=true
    Value=&RigVMFunction_MathBoolToFloat(Value=RigVMVariableNode() {variable=isKneeSlide}.Value).Result
   )
   n9 = &RigVMFunction_AnimEvalRichCurve(Curve=(EditorCurveData=(Keys=((),(Time=0.5503,Value=0.0275),(Time=1,Value=1)))), SourceMinimum=0, TargetMinimum=0, Value=@n8.Result)
   Multiply = &RigVMFunction_MathDoubleMul(A=@"Delta Rotation Z".Delta Rotation, B=@n9.Result)
   > &RigUnit_ModifyTransforms(
    ItemToModify="((Item=(Type=Bone,Name=\"base_cloth_l_bck\"),Transform=(Rotation=(X=0,Y=0,Z=0,W=1),Translation=(X=0,Y=0,Z=9.0695),Scale3D=(X=1,Y=1,Z=1))),(Item=(Type=Bone,Name=\"dyn_cloth_01_bck\"),Transform=(Rotation=(X=0,Y=0,Z=0,W=-1),Translation=(X=0,Y=0,Z=0.8487),Scale3D= ...(truncated)"
    WeightMinimum=0
    Weight=@Multiply.Result
   )
   n10 = RigVMVariableNode() {variable=rigidbodyenable}
   > &RigUnit_ModifyTransforms(
    ItemToModify="((Item=(Type=Bone,Name=\"base_cloth_l_bck\"),Transform=(Rotation=(X=0.0002,Y=-0.0008,Z=-0.2153,W=0.9766),Translation=(X=-2.4261,Y=-1.3572,Z=10.3989),Scale3D=(X=1,Y=1,Z=1))),(Item=(Type=Bone,Name=\"dyn_cloth_01_bck\"),Transform=(Rotation=(X=0,Y=0.0224,Z=0,W=0.9 ...(truncated)"
    WeightMinimum=0
    Weight=&RigVMFunction_MathDoubleMul(
     B=RigVMDispatchNode(False=0, True=1, Condition=@n10.Value) {template="DISPATCH_RigVMDispatch_If(in Condition,in True,in False,out Result)"}.Result
     A=@n9.Result
    ).Result
   )
   RigVMFunctionReturnNode(ExecuteContext=@RigVMFunction_Sequence.D)
   RigVMDispatchNode(Enabled=True, ExecuteContext=(), ScreenColor=(R=1,G=1,B=1,A=1), ScreenDuration=0.05, Value=@RigVMFunction_AnimEvalRichCurve.Result) {template="DISPATCH_RigVMDispatch_Print(in Prefix,in Value,in Enabled,in ScreenDuration,in ScreenColor)"}
  ]}
 New Function_ContainedGraph {nodes [
   n1 = RigVMFunctionEntryNode()
   > RigVMFunction_Sequence = RigVMAggregateNode()
   RigVMFunction_Sequence_1 = RigVMAggregateNode(ExecuteContext=@RigVMFunction_Sequence.A)
   n2 = RigVMDispatchNode(Array=((Type=Bone,Name="upperarm_l"),(Type=Bone,Name="lowerarm_l"),(Type=Bone,Name="hand_l")), ExecuteContext=@RigVMFunction_Sequence_1.A) {template="DISPATCH_RigVMDispatch_ArrayIterator(in Array,out Element,out Index,out Count,out Ratio)"}
   > &RigUnit_SetTransform(
    Weight=1
    bInitial=False
    Item=RigVMDispatchNode(Array=((Type=Bone,Name="deform_upperarm_l"),(Type=Bone,Name="deform_lowerarm_l"),(Type=Bone,Name="deform_hand_l")), Index=@n2.Index) {template="DISPATCH_RigVMDispatch_ArrayGetAtIndex(in Array,in Index,out Element)"}.Element
    Value=&RigUnit_GetTransform(bInitial=False, Item=@n2.Element).Transform
   )
   n3 = /Script/PBIK.RigUnit_PBIK(
    BoneSettings=((Bone="deform_hand_l",RotationStiffness=0,PositionStiffness=0,X=Free,MinX=0,MaxX=0,Y=Free,MinY=0,MaxY=0,Z=Free,MinZ=0,MaxZ=0,bUsePreferredAngles=true,PreferredAngles=(X=1,Y=0,Z=0)))
    Debug=(DrawScale=1,bDrawDebug=False)
    Effectors=((Bone="hand_l",Transform=(Rotation=(X=0,Y=0,Z=0,W=1),Translation=(X=0,Y=0,Z=0),Scale3D=(X=1,Y=1,Z=1)),PositionAlpha=1,RotationAlpha=1,StrengthAlpha=1,PullChainAlpha=1,PinRotation=1,ChainDepth=0))
    ExcludedBones=()
    Root=deform_upperarm_l
    Settings="(Iterations=20,SubIterations=0,MassMultiplier=1,bAllowStretch=false,RootBehavior=PrePull,PrePullRootSettings=(RotationAlpha=0,RotationAlphaX=1,RotationAlphaY=1,RotationAlphaZ=1,PositionAlpha=1,PositionAlphaX=1,PositionAlphaY=1,PositionAlphaZ=1),GlobalPullC ...(truncated)"
    ExecutePin=@n2.Completed
   ) {is=struct RigVMUnitNode}
   RigVMFunction_ControlFlowBranch = &RigVMFunction_ControlFlowBranch(ExecuteContext=@n3.ExecutePin, Condition=@n1.disable_arms_modifications)
   n4 = &RigUnit_SphericalPoseReader(
    ActiveRegionScaleFactors=(PositiveWidth=2,NegativeWidth=0.5,PositiveHeight=1,NegativeHeight=1)
    ActiveRegionSize=0.1
    Debug=(bDrawDebug=false,bDraw2D=False,bDrawLocalAxes=False,DebugScale=25,DebugSegments=20,DebugThickness=0.25)
    DriverAxis=(X=1,Y=0,Z=0)
    DriverItem=(Type=Bone,Name="upperarm_l")
    FalloffRegionScaleFactors=(PositiveWidth=1,NegativeWidth=1,PositiveHeight=2,NegativeHeight=2)
    FalloffSize=0.2
    FlipHeightScaling=False
    FlipWidthScaling=False
    OptionalParentItem=(Type=Bone,Name="spine_03")
    RotationOffset=(X=0,Y=-145,Z=0)
    ExecutePin=@RigVMFunction_ControlFlowBranch.False
   )
   > ApplyFBIKChainCorrection = RigVMFunctionReferenceNode(
    End=(Type=Bone,Name="deform_hand_l")
    QuatAmount=-0.4
    Start=(Type=Bone,Name="deform_upperarm_l")
    Weight=&RigVMFunction_MathDoubleMul(A=@n4.OutputParam, B=RigVMVariableNode() {variable=guanteletEquipped}.Value).Result
   )
   > Get Normalized Distance Between_1 = RigVMFunctionReferenceNode(Debug=false, Max Distance=40, Min Distance=10, Source=(Type=Bone,Name="hand_l"), Target=(Type=Bone,Name="pelvis"))
   > ApplyFBIKChainCorrection_1 = RigVMFunctionReferenceNode(
    End=(Type=Bone,Name="deform_hand_l")
    QuatAmount=-0.8
    Start=(Type=Bone,Name="deform_upperarm_l")
    Weight=&RigVMFunction_MathDoubleMul(A=@"Get Normalized Distance Between_1".Result, B=RigVMVariableNode() {variable=guanteletEquipped}.Value).Result
   )
   n5 = RigVMDispatchNode(Array=((Type=Bone,Name="upperarm_r"),(Type=Bone,Name="lowerarm_r"),(Type=Bone,Name="hand_r")), ExecuteContext=@RigVMFunction_Sequence_1.B) {template="DISPATCH_RigVMDispatch_ArrayIterator(in Array,out Element,out Index,out Count,out Ratio)"}
   > &RigUnit_SetTransform(
    Weight=1
    bInitial=False
    Item=RigVMDispatchNode(Array=((Type=Bone,Name="deform_upperarm_r"),(Type=Bone,Name="deform_lowerarm_r"),(Type=Bone,Name="deform_hand_r")), Index=@n5.Index) {template="DISPATCH_RigVMDispatch_ArrayGetAtIndex(in Array,in Index,out Element)"}.Element
    Value=&RigUnit_GetTransform(bInitial=False, Item=@n5.Element).Transform
   )
   n6 = &RigUnit_PBIK(
    BoneSettings=((Bone="deform_hand_r",RotationStiffness=0,PositionStiffness=0,X=Free,MinX=0,MaxX=0,Y=Free,MinY=0,MaxY=0,Z=Free,MinZ=0,MaxZ=0,bUsePreferredAngles=true,PreferredAngles=(X=1,Y=0,Z=0)))
    Debug=(DrawScale=1,bDrawDebug=False)
    Effectors=((Bone="hand_r",Transform=(Rotation=(X=0,Y=0,Z=0,W=1),Translation=(X=0,Y=0,Z=0),Scale3D=(X=1,Y=1,Z=1)),PositionAlpha=1,RotationAlpha=1,StrengthAlpha=1,PullChainAlpha=1,PinRotation=1,ChainDepth=0))
    ExcludedBones=()
    Root=deform_upperarm_r
    Settings="(Iterations=20,SubIterations=0,MassMultiplier=1,bAllowStretch=false,RootBehavior=PrePull,PrePullRootSettings=(RotationAlpha=0,RotationAlphaX=1,RotationAlphaY=1,RotationAlphaZ=1,PositionAlpha=1,PositionAlphaX=1,PositionAlphaY=1,PositionAlphaZ=1),GlobalPullC ...(truncated)"
    ExecutePin=@n5.Completed
   )
   n7 = &RigVMFunction_ControlFlowBranch(Condition=False, ExecuteContext=@n6.ExecutePin)
   n8 = &RigUnit_SphericalPoseReader(
    ActiveRegionScaleFactors=(PositiveWidth=2,NegativeWidth=0.5,PositiveHeight=1,NegativeHeight=1)
    ActiveRegionSize=0.1
    Debug=(bDrawDebug=false,bDraw2D=False,bDrawLocalAxes=False,DebugScale=25,DebugSegments=20,DebugThickness=0.25)
    DriverAxis=(X=1,Y=0,Z=0)
    DriverItem=(Type=Bone,Name="upperarm_r")
    FalloffRegionScaleFactors=(PositiveWidth=1,NegativeWidth=1,PositiveHeight=2,NegativeHeight=2)
    FalloffSize=0.2
    FlipHeightScaling=False
    FlipWidthScaling=False
    OptionalParentItem=(Type=Bone,Name="spine_03")
    RotationOffset=(X=0,Y=-145,Z=0)
    ExecutePin=@n7.False
   )
   Multiply = &RigVMFunction_MathDoubleMul(A=@n8.OutputParam, B=RigVMVariableNode() {variable=guanteletEquipped}.Value)
   > ApplyFBIKChainCorrection_2 = RigVMFunctionReferenceNode(End=(Type=Bone,Name="deform_hand_r"), QuatAmount=1, Start=(Type=Bone,Name="deform_upperarm_r"), Weight=@Multiply.Result)
   RigVMFunctionReturnNode(ExecuteContext=@RigVMFunction_Sequence.C)
   RigUnit_GetTransform = &RigUnit_GetTransform(Item=(Type=Bone,Name="hand_l"), bInitial=False)
   RigUnit_SetTransform = &RigUnit_SetTransform(Item=(Type=Bone,Name="deform_hand_l"), Weight=1, bInitial=False, Value=@RigUnit_GetTransform.Transform, ExecutePin=@RigVMFunction_Sequence.B)
   Get Normalized Distance Between_2 = RigVMFunctionReferenceNode(Debug=false, ExecuteContext=(), Max Distance=40, Min Distance=10, Source=(Type=Bone,Name="hand_r"), Target=(Type=Bone,Name="pelvis"))
   > ApplyFBIKChainCorrection_3 = RigVMFunctionReferenceNode(
    End=(Type=Bone,Name="deform_hand_r")
    QuatAmount=0
    Start=(Type=Bone,Name="deform_upperarm_r")
    Weight=@"Get Normalized Distance Between_2".Result
   )
   RigVMRerouteNode(Value=@n1.disable_arms_modifications)
   RigVMVariableNode() {variable=l_arm_bones}
   RigVMVariableNode() {variable=l_arm_deform_bones}
  ]}
 New Function_ContainedGraph {nodes [
   n1 = RigVMFunctionEntryNode()
   > RigVMFunction_Sequence = RigVMAggregateNode()
   RigVMFunction_ControlFlowBranch = &RigVMFunction_ControlFlowBranch(Condition=@n1.Debug, ExecuteContext=@RigVMFunction_Sequence.B)
   n2 = &RigUnit_GetTransform(bInitial=False, Item=@n1.Source)
   n3 = &RigUnit_GetTransform(bInitial=False, Item=@n1.Target)
   RigVMFunction_MathVectorDistance = &RigVMFunction_MathVectorDistance(A=@n2.Transform.Translation, B=@n3.Transform.Translation)
   Less = /Script/RigVM.RigVMFunction_MathDoubleLess(B=@n1.Max Distance, A=@RigVMFunction_MathVectorDistance.Result) {is=struct RigVMUnitNode, template="Less::Execute(in A,in B,out Result)"}
   n4 = &RigVMFunction_ControlFlowBranch(Condition=@Less.Result, ExecuteContext=@RigVMFunction_ControlFlowBranch.True)
   RigVMFunction_DebugLineNoSpace = /Script/RigVM.RigVMFunction_DebugLineNoSpace(
    Color=(R=1,G=0,B=0,A=1)
    DebugDrawSettings=(DepthPriority=SDPG_Foreground,Lifetime=-1)
    Thickness=2
    WorldOffset=(Rotation=(X=0,Y=0,Z=0,W=1),Translation=(X=0,Y=0,Z=0),Scale3D=(X=1,Y=1,Z=1))
    bEnabled=True
    ExecutePin=@n4.False
    B=@n3.Transform.Translation
    A=@n2.Transform.Translation
   ) {is=struct RigVMUnitNode}
   Remap = &RigVMFunction_MathDoubleRemap(
    TargetMaximum=1
    TargetMinimum=0
    Value=@RigVMFunction_MathVectorDistance.Result
    SourceMaximum=@n1.Max Distance
    SourceMinimum=@n1.Min Distance
   )
   Subtract = &RigVMFunction_MathDoubleSub(A=1, B=@Remap.Result)
   &RigVMFunction_DebugLineNoSpace(
    Color=(R=0,G=1,B=0,A=1)
    DebugDrawSettings=(DepthPriority=SDPG_Foreground,Lifetime=-1)
    Thickness=2
    WorldOffset=(Rotation=(X=0,Y=0,Z=0,W=1),Translation=(X=0,Y=0,Z=0),Scale3D=(X=1,Y=1,Z=1))
    bEnabled=True
    ExecutePin=@n4.True
    Color.R=@Subtract.Result
    B=@n3.Transform.Translation
    A=@n2.Transform.Translation
    Color.G=@Remap.Result
   )
   RigVMFunctionReturnNode(ExecuteContext=@RigVMFunction_Sequence.C, Result=RigVMVariableNode() {variable=Output}.Value)
   RigVMVariableNode(Value=&RigVMFunction_MathDoubleSub(A=1, B=@Remap.Result).Result, ExecuteContext=@RigVMFunction_Sequence.A) {variable=Output}
  ]}
 New Function_ContainedGraph {nodes [
   RigVMFunctionEntryNode()
   > RigVMFunction_Sequence = RigVMAggregateNode()
   RigVMFunctionReturnNode(ExecuteContext=@RigVMFunction_Sequence.C)
   n1 = /Script/RigVM.RigVMFunction_MathQuaternionToAxisAndAngle(Value=&RigVMFunction_MathQuaternionSwingTwist(
     TwistAxis=(X=-1,Y=0,Z=0)
     Input=&RigUnit_GetTransform(Item=(Type=Bone,Name="deform_upperarm_r"), Space=LocalSpace, bInitial=False).Transform.Rotation
    ).Twist) {is=struct RigVMUnitNode, template="ToAxisAndAngle::Execute(in Value,out Axis,out Angle)"}
   &RigUnit_SetRotation(
    Item=(Type=Bone,Name="deform_upperarm_twist_01_r")
    Weight=1
    bInitial=False
    Value=/Script/RigVM.RigVMFunction_MathQuaternionFromAxisAndAngle(Angle=&RigVMFunction_MathDoubleMul(B=-0.6, A=@n1.Angle).Result, Axis=@n1.Axis) {is=struct RigVMUnitNode, template="FromAxisAndAngle::Execute(in Axis,in Angle,out Result)"}.Result
    ExecutePin=@RigVMFunction_Sequence.B
   )
   > &RigUnit_SetRotation(
    Item=(Type=Bone,Name="deform_upperarm_twist_02_r")
    Weight=1
    bInitial=False
    Value=&RigVMFunction_MathQuaternionFromAxisAndAngle(Angle=&RigVMFunction_MathDoubleMul(B=-0.2, A=@n1.Angle).Result, Axis=@n1.Axis) {template="FromAxisAndAngle::Execute(in Axis,in Angle,out Result)"}.Result
   )
   n2 = &RigUnit_GetTransform(Item=(Type=Bone,Name="lowerarm_twist_01_r"), Space=LocalSpace, bInitial=False)
   > &RigUnit_SetRotation(Item=(Type=Bone,Name="deform_lowerarm_twist_01_r"), Weight=1, bInitial=False, Value=@n2.Transform.Rotation)
   n3 = &RigUnit_GetTransform(Item=(Type=Bone,Name="lowerarm_twist_02_r"), Space=LocalSpace, bInitial=False)
   > &RigUnit_SetRotation(Item=(Type=Bone,Name="deform_lowerarm_twist_02_r"), Weight=0.3, bInitial=False, Value=@n3.Transform.Rotation)
   RigUnit_GetTransform = &RigUnit_GetTransform(Item=(Type=Bone,Name="deform_upperarm_l"), Space=LocalSpace, bInitial=False)
   RigVMFunction_MathQuaternionSwingTwist = &RigVMFunction_MathQuaternionSwingTwist(TwistAxis=(X=-1,Y=0,Z=0), Input=@RigUnit_GetTransform.Transform.Rotation)
   RigVMFunction_MathQuaternionToAxisAndAngle = &RigVMFunction_MathQuaternionToAxisAndAngle(Value=@RigVMFunction_MathQuaternionSwingTwist.Twist) {template="ToAxisAndAngle::Execute(in Value,out Axis,out Angle)"}
   RigVMFunction_MathQuaternionFromAxisAndAngle = &RigVMFunction_MathQuaternionFromAxisAndAngle(
    Angle=&RigVMFunction_MathDoubleMul(B=-0.6, A=@RigVMFunction_MathQuaternionToAxisAndAngle.Angle).Result
    Axis=@RigVMFunction_MathQuaternionToAxisAndAngle.Axis
   ) {template="FromAxisAndAngle::Execute(in Axis,in Angle,out Result)"}
   &RigUnit_SetRotation(
    Item=(Type=Bone,Name="deform_upperarm_twist_01_l")
    Weight=1
    bInitial=False
    Value=@RigVMFunction_MathQuaternionFromAxisAndAngle.Result
    ExecutePin=@RigVMFunction_Sequence.A
   )
   > &RigUnit_SetRotation(
    Item=(Type=Bone,Name="deform_upperarm_twist_02_l")
    Weight=1
    bInitial=False
    Value=&RigVMFunction_MathQuaternionFromAxisAndAngle(
     Angle=&RigVMFunction_MathDoubleMul(B=-0.2, A=@RigVMFunction_MathQuaternionToAxisAndAngle.Angle).Result
     Axis=@RigVMFunction_MathQuaternionToAxisAndAngle.Axis
    ) {template="FromAxisAndAngle::Execute(in Axis,in Angle,out Result)"}.Result
   )
   n4 = &RigUnit_GetTransform(Item=(Type=Bone,Name="lowerarm_twist_01_l"), Space=LocalSpace, bInitial=False)
   > &RigUnit_SetRotation(Item=(Type=Bone,Name="deform_lowerarm_twist_01_l"), Weight=1, bInitial=False, Value=@n4.Transform.Rotation)
   n5 = &RigUnit_GetTransform(Item=(Type=Bone,Name="lowerarm_twist_02_l"), Space=LocalSpace, bInitial=False)
   > &RigUnit_SetRotation(Item=(Type=Bone,Name="deform_lowerarm_twist_02_l"), Weight=0.3, bInitial=False, Value=@n5.Transform.Rotation)
   &RigVMFunction_MathQuaternionSwingTwist(TwistAxis=(X=1,Y=0,Z=0), Input=@n2.Transform.Rotation)
   &RigVMFunction_MathQuaternionSwingTwist(TwistAxis=(X=1,Y=0,Z=0), Input=@n4.Transform.Rotation)
   &RigVMFunction_MathQuaternionSwingTwist(TwistAxis=(X=1,Y=0,Z=0), Input=@n5.Transform.Rotation)
   &RigVMFunction_MathQuaternionSwingTwist(TwistAxis=(X=1,Y=0,Z=0), Input=@n3.Transform.Rotation)
  ]}
 New Function_ContainedGraph {nodes [
   n1 = RigVMFunctionEntryNode()
   n2 = &RigUnit_GetTransform(bInitial=False, Item=@n1.Start)
   Subtract = /Script/RigVM.RigVMFunction_MathVectorSub(B=@n2.Transform.Translation, A=&RigUnit_GetTransform(bInitial=False, Item=@n1.End).Transform.Translation) {is=struct RigVMUnitNode, template="Subtract::Execute(in A,in B,out Result)"}
   Unit = /Script/RigVM.RigVMFunction_MathVectorUnit(Value=@Subtract.Result) {is=struct RigVMUnitNode, template="Unit::Execute(in Value,out Result)"}
   RigVMFunction_MathQuaternionFromAxisAndAngle = &RigVMFunction_MathQuaternionFromAxisAndAngle(Axis=@Unit.Result, Angle=@n1.QuatAmount) {template="FromAxisAndAngle::Execute(in Axis,in Angle,out Result)"}
   Multiply = /Script/RigVM.RigVMFunction_MathQuaternionMul(B=@n2.Transform.Rotation, A=@RigVMFunction_MathQuaternionFromAxisAndAngle.Result) {is=struct RigVMUnitNode, template="Multiply::Execute(in A,in B,out Result)"}
   > &RigUnit_SetRotation(Space=GlobalSpace, bInitial=False, Item=@n1.Start, Weight=@n1.Weight, Value=@Multiply.Result)
   > RigVMFunctionReturnNode()
  ]}
 New Function_ContainedGraph {nodes [
   n1 = RigVMFunctionEntryNode()
   RigVMFunction_ControlFlowBranch = &RigVMFunction_ControlFlowBranch(Condition=@n1.Disable Arms, ExecuteContext=@n1.ExecuteContext)
   RigVMFunctionReturnNode(ExecuteContext=@RigVMFunction_ControlFlowBranch.Completed)
   RigVMFunction_Sequence = RigVMAggregateNode(ExecuteContext=@RigVMFunction_ControlFlowBranch.False)
   ArmIkCorrection = RigVMFunctionReferenceNode(disable_arms_modifications=False, ExecuteContext=@RigVMFunction_Sequence.B)
   DeformTwists = RigVMFunctionReferenceNode(ExecuteContext=@RigVMFunction_Sequence.C)
   n2 = &RigVMFunction_Sequence(ExecuteContext=@RigVMFunction_ControlFlowBranch.True)
   &RigUnit_ParentConstraint(
    AdvancedSettings=(InterpolationType=Average,RotationOrderForFilter=XZY)
    Child=(Type=Bone,Name="deform_upperarm_l")
    Parents=((Item=(Type=Bone,Name="upperarm_l"),Weight=1))
    Weight=1
    ExecutePin=@n2.A
   )
   > &RigUnit_ParentConstraint(
    AdvancedSettings=(InterpolationType=Average,RotationOrderForFilter=XZY)
    Child=(Type=Bone,Name="deform_upperarm_twist_01_l")
    Parents=((Item=(Type=Bone,Name="upperarm_twist_01_l"),Weight=1))
    Weight=1
   )
   > &RigUnit_ParentConstraint(
    AdvancedSettings=(InterpolationType=Average,RotationOrderForFilter=XZY)
    Child=(Type=Bone,Name="deform_upperarm_twist_02_l")
    Parents=((Item=(Type=Bone,Name="upperarm_twist_02_l"),Weight=1))
    Weight=1
   )
   > &RigUnit_ParentConstraint(
    AdvancedSettings=(InterpolationType=Average,RotationOrderForFilter=XZY)
    Child=(Type=Bone,Name="deform_lowerarm_l")
    Parents=((Item=(Type=Bone,Name="lowerarm_l"),Weight=1))
    Weight=1
   )
   > &RigUnit_ParentConstraint(
    AdvancedSettings=(InterpolationType=Average,RotationOrderForFilter=XZY)
    Child=(Type=Bone,Name="deform_lowerarm_twist_01_l")
    Parents=((Item=(Type=Bone,Name="lowerarm_twist_01_l"),Weight=1))
    Weight=1
   )
   > &RigUnit_ParentConstraint(
    AdvancedSettings=(InterpolationType=Average,RotationOrderForFilter=XZY)
    Child=(Type=Bone,Name="deform_lowerarm_twist_02_l")
    Parents=((Item=(Type=Bone,Name="lowerarm_twist_02_l"),Weight=1))
    Weight=1
   )
   > &RigUnit_ParentConstraint(
    AdvancedSettings=(InterpolationType=Average,RotationOrderForFilter=XZY)
    Child=(Type=Bone,Name="deform_hand_l")
    Parents=((Item=(Type=Bone,Name="hand_l"),Weight=1))
    Weight=1
   )
   &RigUnit_ParentConstraint(
    AdvancedSettings=(InterpolationType=Average,RotationOrderForFilter=XZY)
    Child=(Type=Bone,Name="deform_upperarm_r")
    Parents=((Item=(Type=Bone,Name="upperarm_r"),Weight=1))
    Weight=1
    ExecutePin=@n2.B
   )
   > &RigUnit_ParentConstraint(
    AdvancedSettings=(InterpolationType=Average,RotationOrderForFilter=XZY)
    Child=(Type=Bone,Name="deform_upperarm_twist_01_r")
    Parents=((Item=(Type=Bone,Name="upperarm_twist_01_r"),Weight=1))
    Weight=1
   )
   > &RigUnit_ParentConstraint(
    AdvancedSettings=(InterpolationType=Average,RotationOrderForFilter=XZY)
    Child=(Type=Bone,Name="deform_upperarm_twist_02_r")
    Parents=((Item=(Type=Bone,Name="upperarm_twist_02_r"),Weight=1))
    Weight=1
   )
   > &RigUnit_ParentConstraint(
    AdvancedSettings=(InterpolationType=Average,RotationOrderForFilter=XZY)
    Child=(Type=Bone,Name="deform_lowerarm_r")
    Parents=((Item=(Type=Bone,Name="lowerarm_r"),Weight=1))
    Weight=1
   )
   > &RigUnit_ParentConstraint(
    AdvancedSettings=(InterpolationType=Average,RotationOrderForFilter=XZY)
    Child=(Type=Bone,Name="deform_lowerarm_twist_01_r")
    Parents=((Item=(Type=Bone,Name="lowerarm_twist_01_r"),Weight=1))
    Weight=1
   )
   > &RigUnit_ParentConstraint(
    AdvancedSettings=(InterpolationType=Average,RotationOrderForFilter=XZY)
    Child=(Type=Bone,Name="deform_lowerarm_twist_02_r")
    Parents=((Item=(Type=Bone,Name="lowerarm_twist_02_r"),Weight=1))
    Weight=1
   )
   > &RigUnit_ParentConstraint(
    AdvancedSettings=(InterpolationType=Average,RotationOrderForFilter=XZY)
    Child=(Type=Bone,Name="deform_hand_r")
    Parents=((Item=(Type=Bone,Name="hand_r"),Weight=1))
    Weight=1
   )
  ]}
 New Function_ContainedGraph {nodes [
   RigVMFunctionEntryNode()
   > RigVMFunction_Sequence = RigVMAggregateNode()
   &RigUnit_ParentConstraint(
    AdvancedSettings=(RotationOrderForFilter=XZY)
    Child=(Type=Bone,Name="deform_deform_elbow_in_base_l")
    Parents=((Item=(Type=Bone,Name="deform_lowerarm_l"),Weight=0.4),(Item=(Type=Bone,Name="deform_upperarm_l"),Weight=0.6))
    Weight=1
    ExecutePin=@RigVMFunction_Sequence.A
   )
   > &RigUnit_ModifyTransforms(
    ItemToModify=((Item=(Type=Bone,Name="deform_deform_elbow_in_l"),Transform=(Rotation=(X=0,Y=0,Z=0,W=1),Translation=(X=0,Y=9,Z=0),Scale3D=(X=1,Y=1,Z=1))))
    Weight=RigVMDispatchNode(BoneName=root, Default=1, Name=deformRig_l_elbow_alpha) {template="DISPATCH_RigDispatch_GetAnimAttribute(in Name,in BoneName,in Default,out Value,out Found)"}.Value
   )
   &RigUnit_ParentConstraint(
    AdvancedSettings=(RotationOrderForFilter=XZY)
    Child=(Type=Bone,Name="deform_deform_elbow_out_base_l")
    Parents=((Item=(Type=Bone,Name="deform_lowerarm_l"),Weight=0.3),(Item=(Type=Bone,Name="deform_upperarm_l"),Weight=0.7))
    Weight=1
    ExecutePin=@RigVMFunction_Sequence.B
   )
   > &RigUnit_ModifyTransforms(
    ItemToModify=((Item=(Type=Bone,Name="deform_deform_elbow_out_l"),Transform=(Rotation=(X=0,Y=0,Z=0,W=1),Translation=(X=0,Y=-2,Z=0),Scale3D=(X=1,Y=1,Z=1))))
    Weight=RigVMDispatchNode(BoneName=root, Default=1, Name=deformRig_l_elbow_alpha) {template="DISPATCH_RigDispatch_GetAnimAttribute(in Name,in BoneName,in Default,out Value,out Found)"}.Value
   )
   &RigUnit_ParentConstraint(
    AdvancedSettings=(RotationOrderForFilter=XZY)
    Child=(Type=Bone,Name="deform_deform_elbow_in_base_r")
    Parents=((Item=(Type=Bone,Name="deform_lowerarm_r"),Weight=0.4),(Item=(Type=Bone,Name="deform_upperarm_r"),Weight=0.6))
    Weight=1
    ExecutePin=@RigVMFunction_Sequence.C
   )
   > &RigUnit_ModifyTransforms(
    ItemToModify=((Item=(Type=Bone,Name="deform_deform_elbow_in_r"),Transform=(Rotation=(X=0,Y=0,Z=0,W=1),Translation=(X=0,Y=-9,Z=0),Scale3D=(X=1,Y=1,Z=1))))
    Weight=RigVMDispatchNode(BoneName=root, Default=1, Name=deformRig_r_elbow_alpha) {template="DISPATCH_RigDispatch_GetAnimAttribute(in Name,in BoneName,in Default,out Value,out Found)"}.Value
   )
   &RigUnit_ParentConstraint(
    AdvancedSettings=(RotationOrderForFilter=XZY)
    Child=(Type=Bone,Name="deform_deform_elbow_out_base_r")
    Parents=((Item=(Type=Bone,Name="deform_lowerarm_r"),Weight=0.3),(Item=(Type=Bone,Name="deform_upperarm_r"),Weight=0.7))
    Weight=1
    ExecutePin=@RigVMFunction_Sequence.D
   )
   > &RigUnit_ModifyTransforms(
    ItemToModify=((Item=(Type=Bone,Name="deform_deform_elbow_out_r"),Transform=(Rotation=(X=0,Y=0,Z=0,W=1),Translation=(X=0,Y=2,Z=0),Scale3D=(X=1,Y=1,Z=1))))
    Weight=RigVMDispatchNode(BoneName=root, Default=1, Name=deformRig_r_elbow_alpha) {template="DISPATCH_RigDispatch_GetAnimAttribute(in Name,in BoneName,in Default,out Value,out Found)"}.Value
   )
   RigVMFunctionReturnNode(ExecuteContext=@RigVMFunction_Sequence.E)
  ]}
 New Function_ContainedGraph {nodes [
   n1 = RigVMFunctionEntryNode()
   RigVMFunction_MathQuaternionToEuler = &RigVMFunction_MathQuaternionToEuler(
    RotationOrder=XYZ
    Value=&RigUnit_GetRelativeTransformForItem(
     Child=(Type=Bone,Name="hand_l")
     Parent=(Type=Null,Name="lowerarm_l")
     bChildInitial=False
     bParentInitial=False
     Child.Name=@n1.Child
     Parent.Name=@n1.Parent
    ).RelativeTransform.Rotation
   )
   Remap = &RigVMFunction_MathDoubleRemap(
    TargetMaximum=1
    TargetMinimum=0
    Value=@RigVMFunction_MathQuaternionToEuler.Result.Z
    SourceMaximum=@n1.Source Maximum
    SourceMinimum=@n1.Source Minimum
   )
   > RigVMFunctionReturnNode(Delta Rotation=@Remap.Result)
  ]}
 New Function_ContainedGraph {nodes [
   n1 = RigVMFunctionEntryNode()
   RigVMFunction_Sequence = &RigVMFunction_Sequence(ExecuteContext=@n1.ExecuteContext)
   Delta Rotation Z = RigVMFunctionReferenceNode(Child=thigh_l, Parent=thigh_l_null, Source Maximum=50, Source Minimum=5, ExecuteContext=@RigVMFunction_Sequence.A)
   > RigVMVariableNode(Value=@"Delta Rotation Z".Delta Rotation) {variable=thigh_L_Back}
   > Delta Rotation Z_1 = RigVMFunctionReferenceNode(Child=thigh_r, Parent=thigh_r_null, Source Maximum=50, Source Minimum=5)
   > RigVMVariableNode(Value=@"Delta Rotation Z_1".Delta Rotation) {variable=thigh_R_Back}
   RigVMFunctionReturnNode(ExecuteContext=@RigVMFunction_Sequence.B)
  ]}
 New Function_ContainedGraph {nodes [
   n1 = RigVMFunctionEntryNode()
   RigVMFunction_Sequence = &RigVMFunction_Sequence(ExecuteContext=@n1.ExecuteContext)
   Delta Rotation Y = RigVMFunctionReferenceNode(Child=hand_r, Parent=hand_r_null, Source Maximum=80, ExecuteContext=@RigVMFunction_Sequence.A)
   > &RigUnit_ModifyTransforms(
    ItemToModify=((Item=(Type=Bone,Name="dyn_sleeve_in_r"),Transform=(Rotation=(X=0.1194,Y=0,Z=0,W=0.9928),Translation=(X=0,Y=-1.3452,Z=-6.6819),Scale3D=(X=1,Y=1,Z=1))))
    WeightMinimum=0
    Weight=@"Delta Rotation Y".Delta Rotation
   )
   > Delta Rotation Y_1 = RigVMFunctionReferenceNode(Child=hand_r, Parent=hand_r_null, Source Maximum=-80)
   > &RigUnit_ModifyTransforms(
    ItemToModify=((Item=(Type=Bone,Name="dyn_sleeve_out_r"),Transform=(Rotation=(X=0.4136,Y=0,Z=0,W=-0.9104),Translation=(X=0,Y=2.5299,Z=-10.3732),Scale3D=(X=1,Y=1,Z=1))))
    WeightMinimum=0
    Weight=@"Delta Rotation Y_1".Delta Rotation
   )
   > Delta Rotation Y_2 = RigVMFunctionReferenceNode(Child=hand_l, Parent=hand_l_null, Source Maximum=80)
   > &RigUnit_ModifyTransforms(
    ItemToModify=((Item=(Type=Bone,Name="dyn_sleeve_in_l"),Transform=(Rotation=(X=-0.1194,Y=0,Z=0,W=0.9928),Translation=(X=0,Y=1.3452,Z=-6.6819),Scale3D=(X=1,Y=1,Z=1))))
    WeightMinimum=0
    Weight=@"Delta Rotation Y_2".Delta Rotation
   )
   > Delta Rotation Y_3 = RigVMFunctionReferenceNode(Child=hand_l, Parent=hand_l_null, Source Maximum=-80)
   > &RigUnit_ModifyTransforms(
    ItemToModify=((Item=(Type=Bone,Name="dyn_sleeve_out_l"),Transform=(Rotation=(X=0.4136,Y=0,Z=0,W=0.9104),Translation=(X=0,Y=-2.5299,Z=-10.3732),Scale3D=(X=1,Y=1,Z=1))))
    WeightMinimum=0
    Weight=@"Delta Rotation Y_3".Delta Rotation
   )
   > &RigUnit_ModifyTransforms(
    ItemToModify="((Item=(Type=Bone,Name=\"dyn_sleeve_out_l\"),Transform=(Rotation=(X=0.2016,Y=0,Z=0,W=0.9795),Translation=(X=-2.324,Y=-3.7288,Z=-9.4811),Scale3D=(X=1,Y=1,Z=1))),(Item=(Type=Bone,Name=\"dyn_sleeve_in_l\"),Transform=(Rotation=(X=0,Y=0,Z=0,W=-1),Translation=(X=0,Y ...(truncated)"
    WeightMinimum=0
    Weight=RigVMDispatchNode(BoneName=root, Default=1, Name=deformRig_l_wrist_up_alpha) {template="DISPATCH_RigDispatch_GetAnimAttribute(in Name,in BoneName,in Default,out Value,out Found)"}.Value
   )
   RigVMFunctionReturnNode(ExecuteContext=@RigVMFunction_Sequence.B)
  ]}
 New Function_ContainedGraph {nodes [
   n1 = RigVMFunctionEntryNode()
   RigVMFunction_ControlFlowBranch = &RigVMFunction_ControlFlowBranch(Condition=true, ExecuteContext=@n1.ExecuteContext)
   RigVMFunctionReturnNode(ExecuteContext=@RigVMFunction_ControlFlowBranch.Completed)
   RigVMFunction_MathVectorDistance = &RigVMFunction_MathVectorDistance(
    A=&RigUnit_GetTransform(Item=(Type=Bone,Name="dyn_cloth_01_l_bck"), bInitial=False).Transform.Translation
    B=&RigUnit_GetTransform(Item=(Type=Bone,Name="foot_r"), bInitial=False).Transform.Translation
   )
   Remap = &RigVMFunction_MathDoubleRemap(SourceMaximum=48, SourceMinimum=50, TargetMaximum=1, TargetMinimum=0, Value=@RigVMFunction_MathVectorDistance.Result)
   &RigUnit_ModifyTransforms(
    ItemToModify="((Item=(Type=Bone,Name=\"base_cloth_bck\"),Transform=(Rotation=(X=0,Y=-0.3007,Z=0,W=0.9537),Translation=(X=-25,Y=0,Z=10),Scale3D=(X=1,Y=1,Z=1))),(Item=(Type=Bone,Name=\"base_cloth_l_bck\"),Transform=(Rotation=(X=0,Y=-0.3007,Z=0,W=0.9537),Translation=(X=-25,Y=0 ...(truncated)"
    WeightMinimum=0
    Weight=RigVMDispatchNode(False=0, True=@Remap.Result, Condition=RigVMVariableNode() {variable=isSliding}.Value) {template="DISPATCH_RigVMDispatch_If(in Condition,in True,in False,out Result)"}.Result
    ExecutePin=@RigVMFunction_ControlFlowBranch.True
   )
   > &RigUnit_ModifyTransforms(
    ItemToModify="((Item=(Type=Bone,Name=\"base_cloth_bck\"),Transform=(Rotation=(X=0,Y=0,Z=0,W=1),Translation=(X=-10,Y=0,Z=10),Scale3D=(X=1,Y=1,Z=1))),(Item=(Type=Bone,Name=\"base_cloth_l_bck\"),Transform=(Rotation=(X=0,Y=0,Z=0,W=1),Translation=(X=-10,Y=0,Z=10),Scale3D=(X=1,Y= ...(truncated)"
    WeightMinimum=0
    Weight=RigVMDispatchNode(Condition=False, False=0, True=@Remap.Result) {template="DISPATCH_RigVMDispatch_If(in Condition,in True,in False,out Result)"}.Result
   )
   RigVMFunction_MathQuaternionFromEuler = &RigVMFunction_MathQuaternionFromEuler(Euler=(X=0,Y=-90,Z=25))
   &RigVMFunction_MathQuaternionFromEuler(Euler=(X=25,Y=-90,Z=0))
   RigVMVariableNode() {variable=isSliding}
  ]}
 RigVMFunctionLibrary {nodes [
   ApplyFBIKChainCorrection = RigVMCollapseNode(End=(Type=None,Name="None"), ExecuteContext=(), Start=(Type=None,Name="None"))
   ArmIkCorrection = RigVMCollapseNode(ExecuteContext=(), disable_arms_modifications=False)
   Constraints = RigVMCollapseNode(ExecuteContext=())
   DeformTwists = RigVMCollapseNode(ExecuteContext=())
   Delta Rotation Y = RigVMCollapseNode(ExecuteContext=())
   Delta Rotation Z = RigVMCollapseNode(ExecuteContext=())
   Delta Rotation Z_1 = RigVMCollapseNode(ExecuteContext=())
   Elbow Correction = RigVMCollapseNode(ExecuteContext=())
   Get Normalized Distance Between = RigVMCollapseNode(ExecuteContext=(), Min Distance=0, Source=(Type=None,Name="None"), Target=(Type=None,Name="None"))
   Jacket Setup = RigVMCollapseNode(ExecuteContext=())
   Sleeves Setup = RigVMCollapseNode(ExecuteContext=())
   TatToon_IK = RigVMCollapseNode(Disable Arms=False, Disable Legs=False, ExecuteContext=())
   Variables = RigVMCollapseNode(ExecuteContext=())
   isSliding = RigVMCollapseNode(ExecuteContext=())
  ]}
 RigVMFunction_Sequence_1_1_ContainedGraph {nodes [
   n1 = RigVMFunctionEntryNode()
   n2 = &RigVMFunction_Sequence(ExecuteContext=@n1.ExecuteContext)
   n3 = &RigVMFunction_Sequence(ExecuteContext=@n2.B)
   RigVMFunctionReturnNode(A=@n2.A, B=@n3.A, C=@n3.B)
  ]}
 RigVMFunction_Sequence_1_ContainedGraph {nodes [
   n1 = RigVMFunctionEntryNode()
   RigVMFunction_Sequence = &RigVMFunction_Sequence(ExecuteContext=@n1.ExecuteContext)
   n2 = &RigVMFunction_Sequence(ExecuteContext=@RigVMFunction_Sequence.B)
   RigVMFunctionReturnNode(A=@RigVMFunction_Sequence.A, B=@n2.A, C=@n2.B)
  ]}
 RigVMFunction_Sequence_1_ContainedGraph {nodes [
   n1 = RigVMFunctionEntryNode()
   RigVMFunction_Sequence = &RigVMFunction_Sequence(ExecuteContext=@n1.ExecuteContext)
   n2 = &RigVMFunction_Sequence(ExecuteContext=@RigVMFunction_Sequence.B)
   n3 = &RigVMFunction_Sequence(ExecuteContext=@n2.B)
   RigVMFunctionReturnNode(A=@RigVMFunction_Sequence.A, B=@n2.A, C=@n3.A, D=@n3.B)
  ]}
 RigVMFunction_Sequence_1_ContainedGraph {nodes [
   n1 = RigVMFunctionEntryNode()
   RigVMFunction_Sequence = &RigVMFunction_Sequence(ExecuteContext=@n1.ExecuteContext)
   n2 = &RigVMFunction_Sequence(ExecuteContext=@RigVMFunction_Sequence.B)
   RigVMFunctionReturnNode(A=@RigVMFunction_Sequence.A, B=@n2.A, C=@n2.B)
  ]}
 RigVMFunction_Sequence_1_ContainedGraph {nodes [
   n1 = RigVMFunctionEntryNode()
   RigVMFunction_Sequence = &RigVMFunction_Sequence(ExecuteContext=@n1.ExecuteContext)
   n2 = &RigVMFunction_Sequence(ExecuteContext=@RigVMFunction_Sequence.B)
   RigVMFunctionReturnNode(A=@RigVMFunction_Sequence.A, B=@n2.A, C=@n2.B)
  ]}
 RigVMFunction_Sequence_1_ContainedGraph {nodes [
   n1 = RigVMFunctionEntryNode()
   RigVMFunction_Sequence = &RigVMFunction_Sequence(ExecuteContext=@n1.ExecuteContext)
   n2 = &RigVMFunction_Sequence(ExecuteContext=@RigVMFunction_Sequence.B)
   RigVMFunctionReturnNode(A=@RigVMFunction_Sequence.A, B=@n2.A, C=@n2.B)
  ]}
 RigVMFunction_Sequence_1_ContainedGraph {nodes [
   n1 = RigVMFunctionEntryNode()
   RigVMFunction_Sequence = &RigVMFunction_Sequence(ExecuteContext=@n1.ExecuteContext)
   n2 = &RigVMFunction_Sequence(ExecuteContext=@RigVMFunction_Sequence.B)
   RigVMFunctionReturnNode(A=@RigVMFunction_Sequence.A, B=@n2.A, C=@n2.B)
  ]}
 RigVMFunction_Sequence_1_ContainedGraph {nodes [
   n1 = RigVMFunctionEntryNode()
   RigVMFunction_Sequence = &RigVMFunction_Sequence(ExecuteContext=@n1.ExecuteContext)
   n2 = &RigVMFunction_Sequence(ExecuteContext=@RigVMFunction_Sequence.B)
   n3 = &RigVMFunction_Sequence(ExecuteContext=@n2.B)
   n4 = &RigVMFunction_Sequence(ExecuteContext=@n3.B)
   RigVMFunctionReturnNode(A=@RigVMFunction_Sequence.A, B=@n2.A, C=@n3.A, D=@n4.A, E=@n4.B)
  ]}
 RigVMModel {nodes [
   RigUnit_BeginExecution = &RigUnit_BeginExecution()
   RigVMFunction_Sequence = RigVMAggregateNode(ExecuteContext=@RigUnit_BeginExecution.ExecutePin)
   n1 = &RigVMFunction_Sequence(ExecuteContext=@RigVMFunction_Sequence.C)
   FN_DeformRig_FNC = RigVMFunctionReferenceNode(
    DeformRig_ Curve="(deformRig_l_upperarm_up_alpha_72_#1=true,deformRig_r_upperarm_up_alpha_73_#2=true,deformRig_l_upperarm_fr_alpha_74_#3=False,deformRig_r_upperarm_fr_alpha_75_CB51B80C ...(truncated)"
    DeformRig_Bool="(upperarm_1_#4=True,elbow_3_#5=True,wrist_7_#6=True,groin_8_#7=True,glute_17_#8=True,knee_9_40C408324CA261 ...(truncated)"
    DeformRig_Debug=(upperarm_debug_18_#4=False,wrist_debug_20_#6=False,lat_debug_26_#9=False,groinGlute_debug_27_#7=False)
    DeformRig_Value="(upperarm_up_18_#10=(Rotation=(X=0.0,Y=-0.0,Z=0.0,W=1.0),Translation=(X=0.0,Y=0.0,Z=8.0),Scale3D=(X=0.70,Y=1.0,Z=1.0)),upperarm_fwd_fr_19_#11=(Rotation=(X=0.0,Y=-0.0,Z=0.0,W=1.0),Translation=(X=-6,Y ...(truncated)"
    ExecuteContext=@n1.A
   )
   > Variables = RigVMFunctionReferenceNode()
   > Constraints = RigVMFunctionReferenceNode()
   > Jacket Setup = RigVMFunctionReferenceNode()
   > Sleeves Setup = RigVMFunctionReferenceNode()
   > Elbow Correction = RigVMFunctionReferenceNode()
   > n2 = &RigUnit_SphericalPoseReader(
    ActiveRegionScaleFactors=(PositiveWidth=1,NegativeWidth=1,PositiveHeight=1,NegativeHeight=0.8)
    ActiveRegionSize=0.1
    Debug=(bDrawDebug=false,bDraw2D=False,bDrawLocalAxes=False,DebugScale=25,DebugSegments=20,DebugThickness=0.25)
    DriverAxis=(X=1,Y=0,Z=0)
    DriverItem=(Type=Bone,Name="upperarm_r")
    FalloffRegionScaleFactors=(PositiveWidth=1,NegativeWidth=1,PositiveHeight=1.7,NegativeHeight=0.5)
    FalloffSize=0.2
    FlipHeightScaling=False
    FlipWidthScaling=False
    OptionalParentItem=(Type=Bone,Name="None")
    RotationOffset=(X=51.9395,Y=8.7023,Z=22.4256)
   )
   > &RigUnit_ModifyTransforms(
    ItemToModify=((Item=(Type=Bone,Name="deform_upperarm_bk_r"),Transform=(Rotation=(X=0,Y=0,Z=0,W=1),Translation=(X=1.8145,Y=6.411,Z=0),Scale3D=(X=1,Y=1,Z=1))))
    WeightMinimum=0
    Weight=@n2.OutputParam
   )
   > &RigUnit_ModifyTransforms(
    ItemToModify="((Item=(Type=Bone,Name=\"deform_upperarm_twist_02_r\"),Transform=(Rotation=(X=0,Y=-0.1786,Z=0,W=-0.9839),Translation=(X=0.6265,Y=0,Z=-3.4504),Scale3D=(X=1,Y=1,Z=1.2099))),(Item=(Type=Bone,Name=\"upperarm_twist_01_r\"),Transform=(Rotation=(X=0,Y=0,Z=0,W=1),Tran ...(truncated)"
    WeightMinimum=0
    Weight=RigVMDispatchNode(BoneName=root, Default=1, Name=deformRig_r_upperarm_up_alpha) {template="DISPATCH_RigDispatch_GetAnimAttribute(in Name,in BoneName,in Default,out Value,out Found)"}.Value
   )
   > &RigUnit_ParentConstraint(
    AdvancedSettings=(InterpolationType=Average,RotationOrderForFilter=XZY)
    Child=(Type=Bone,Name="dyn_tube_03_m")
    Filter=(TranslationFilter=(bX=false,bY=false,bZ=false),RotationFilter=(bX=True,bY=True,bZ=True),ScaleFilter=(bX=false,bY=false,bZ=false))
    Parents=((Item=(Type=Bone,Name="spine_04"),Weight=1))
    Weight=1
   )
   > &RigUnit_ParentConstraint(
    AdvancedSettings=(InterpolationType=Average,RotationOrderForFilter=XZY)
    Child=(Type=Bone,Name="dyn_tube_02_m")
    Filter=(TranslationFilter=(bX=false,bY=false,bZ=false),RotationFilter=(bX=True,bY=True,bZ=True),ScaleFilter=(bX=false,bY=false,bZ=false))
    Parents=((Item=(Type=Bone,Name="spine_04"),Weight=1))
    Weight=1
   )
   n3 = &RigVMFunction_AlphaInterp(
    InterpSpeedDecreasing=10
    InterpSpeedIncreasing=10
    Scale=1
    bClampResult=False
    bInterpResult=False
    Value=&RigVMFunction_MathBoolToFloat(Value=RigVMVariableNode() {variable=isCrouching}.Value).Result
   )
   n4 = &RigVMFunction_AnimEvalRichCurve(
    Curve="(EditorCurveData=(Keys=((InterpMode=RCIM_Linear,TangentMode=RCTM_Auto,TangentWeightMode=RCTWM_WeightedNone,Time=0,Value=0,ArriveTangent=0,ArriveTangentWeight=0,LeaveTangent=0,LeaveTangentWeight=0),(InterpMode=RCIM_Linear,TangentMode=RCTM_Auto,TangentWeight ...(truncated)"
    SourceMinimum=0
    TargetMinimum=0
    Value=@n3.Result
   )
   > &RigUnit_ModifyTransforms(
    ItemToModify="((Item=(Type=Bone,Name=\"dyn_tube_01_m\"),Transform=(Rotation=(X=0,Y=0,Z=0,W=-1),Translation=(X=0,Y=0,Z=0.201),Scale3D=(X=1,Y=1,Z=1))),(Item=(Type=Bone,Name=\"dyn_tube_02_m\"),Transform=(Rotation=(X=0,Y=0,Z=0,W=-1),Translation=(X=0,Y=0,Z=0.5671),Scale3D=(X=1,Y ...(truncated)"
    WeightMinimum=0
    Weight=@n4.Result
   )
   isSliding = RigVMFunctionReferenceNode(ExecuteContext=@n1.B)
   RigVMFunction_MathBoolOr = &RigVMFunction_MathBoolOr(A=RigVMVariableNode() {variable=isEmoting}.Value, B=RigVMVariableNode() {variable=isManequin}.Value)
   TatToon_IK = RigVMFunctionReferenceNode(Disable Legs=false, Disable Arms=@RigVMFunction_MathBoolOr.Result, ExecuteContext=@RigVMFunction_Sequence.A)
   &RigUnit_PrepareForExecution()
   RigUnit_GetRelativeTransformForItem = &RigUnit_GetRelativeTransformForItem(Child=(Type=Bone,Name="pelvis"), Parent=(Type=Bone,Name="root"), bChildInitial=False, bParentInitial=False)
   n5 = &RigVMFunction_AnimEvalRichCurve(
    Curve=(EditorCurveData=(Keys=((InterpMode=RCIM_Cubic,TangentMode=RCTM_Break),(Time=0.6,Value=1),(InterpMode=RCIM_Cubic,TangentMode=RCTM_Break,Time=1))))
    SourceMinimum=0
    TargetMinimum=0
    Value=&RigVMFunction_MathDoubleRemap(
     SourceMaximum=44
     SourceMinimum=60
     TargetMaximum=0
     TargetMinimum=1
     Value=@RigUnit_GetRelativeTransformForItem.RelativeTransform.Translation.Z
    ).Result
   )
   RigVMVariableNode(ExecuteContext=(), Value=@n5.Result) {variable=isCrouchWalkin}
   addFKCtrl = RigVMFunctionReferenceNode(
    Custom_FKCtrl=()
    ExecuteContext=()
    Options_FKCtrl="(pelvis_1_#12=False,spine_3_#13=False,clavicle_5_#14=False,arm_7_#15=true,leg_9_#16=true,head_11_46DC15EF465CD2 ...(truncated)"
    Shape_Settings="(pelvis_2_#17=(bVisible=True,Name=\"Hexagon_Thin\",Color=(R=0.0117,G=0.8281,B=1,A=1),Transform=(Rotation=(X=0,Y=-0.7071,Z=0,W=0.7071),Translation=(X=2,Y=0,Z=0),Scale3D=(X=5,Y=5,Z=7))),spine_4_#18=(bVi ...(truncated)"
   )
   setFKCtrl = RigVMFunctionReferenceNode(ExecuteContext=(), setFK_Ctrl=false)
  ]}
]
hierarchy [
 root {loc=<0 0 0>, rot=<0 0 0 1>}
 pelvis {parent=root, loc=<0 -0.7071 93.1191>, rot=<0.0127 0.707 -0.0127 -0.707>}
 spine_01 {parent=pelvis, loc=<6.7494 -0.6388 0>, rot=<0 0 0.034 -0.9994>}
 spine_02 {parent=spine_01, loc=<7.2518 -0.035 0>, rot=<0 0 0.0238 -0.9997>}
 spine_03 {parent=spine_02, loc=<6.0104 0 0>, rot=<0 0 0.0149 -0.9999>}
 spine_04 {parent=spine_03, loc=<8.1024 0 0>, rot=<0 0 -0.0132 -0.9999>}
 spine_05 {parent=spine_04, loc=<10.0146 0 0>, rot=<0 0 -0.0015 -1>}
 neck_01 {parent=spine_05, loc=<11.3138 0.1529 0>, rot=<0 0 -0.2003 -0.9797>}
 neck_02 {parent=neck_01, loc=<4.5507 0.003 0.0002>, rot=<-0.0402 0.0187 0.0711 -0.9965>}
 head {parent=neck_02, loc=<5.3884 -0.0067 0.0006>, rot=<0.0052 -0.0005 0.1117 -0.9937>}
 faceAttach {parent=head, loc=<-4.146 -0.8545 0>, rot=<0.0029 -0.7071 -0.0029 -0.7071>}
 C_brow_mid {parent=faceAttach, loc=<0 8.1598 4.6104>, rot=<0 0 0 -1>}
 L_brow_mid {parent=faceAttach, loc=<2.8904 7.9296 5.3259>, rot=<-0.0001 -0.0042 0.2553 -0.9669>}
 L_brow_outer {parent=faceAttach, loc=<5.1605 5.8651 4.6582>, rot=<-0.0287 -0.04 0.4491 -0.8921>}
 L_cheek_inner {parent=faceAttach, loc=<3.7048 6.7635 -0.1198>, rot=<-0.0706 -0.2697 0.3184 -0.906>}
 C_lip_upper_mid {parent=faceAttach, loc=<0.0022 8.8861 -2.1487>, rot=<0.2225 0.0013 -0.0005 -0.9749>}
 L_lip_upper_outer {parent=faceAttach, loc=<1.3923 8.41 -2.1797>, rot=<0.0359 -0.0728 0.2627 -0.9615>}
 L_lip_corner {parent=faceAttach, loc=<2.3267 7.4634 -2.5915>, rot=<0.003 0.008 -0.0146 -0.9999>}
 C_nose_bridge {parent=faceAttach, loc=<0 9.8506 1.2827>, rot=<-0.2218 0 0 -0.9751>}
 R_brow_mid {parent=faceAttach, loc=<-2.8904 7.9296 5.3259>, rot=<-0.0001 0.0042 -0.2553 -0.9669>}
 R_brow_outer {parent=faceAttach, loc=<-5.1605 5.8651 4.6582>, rot=<-0.0287 0.04 -0.4491 -0.8921>}
 R_cheek_inner {parent=faceAttach, loc=<-3.7048 6.7635 -0.1198>, rot=<-0.0706 0.2697 -0.3184 -0.906>}
 R_lip_corner {parent=faceAttach, loc=<-2.3267 7.4634 -2.5915>, rot=<0.003 -0.008 0.0146 -0.9999>}
 C_jaw {parent=faceAttach, loc=<-0.0134 -0.1613 -1.4291>, rot=<0 0 0 1>}
 L_lip_lower_outer {parent=C_jaw, loc=<1.3888 8.3831 -1.4241>, rot=<0.013 0.0445 0.3057 -0.951>}
 C_lip_lower_mid {parent=C_jaw, loc=<0.0223 8.7898 -1.4818>, rot=<0 0 0 -1>}
 R_lip_lower_outer {parent=C_jaw, loc=<-1.3619 8.3831 -1.4241>, rot=<0.013 -0.0445 -0.3057 -0.951>}
 teeth_lower {parent=C_jaw, loc=<0.0134 6.0835 -1.7637>, rot=<0 0 0 1>}
 R_eye {parent=faceAttach, loc=<-6.0769 11.1349 9.0435>, rot=<0 0 0.9998 -0.0175>}
 R_eye_lid_upper_mid {parent=R_eye, loc=<0 0 0>, rot=<0.0523 0 0 -0.9986>}
 R_eye_lid_lower_mid {parent=R_eye, loc=<0 0 0>, rot=<0.0329 0 0 -0.9995>}
 L_eye {parent=faceAttach, loc=<0 2.5076 -147.8813>, rot=<0 0 0.9998 0.0175>}
 L_eye_lid_upper_mid {parent=L_eye, loc=<0 0 0>, rot=<0.0523 0 0 -0.9986>}
 L_eye_lid_lower_mid {parent=L_eye, loc=<0 0 0>, rot=<0.0329 0 0 -0.9995>}
 R_lip_upper_outer {parent=faceAttach, loc=<-1.3923 8.41 -2.1797>, rot=<0.0359 0.0728 -0.2627 -0.9615>}
 attach_cape {parent=spine_05, loc=<9.0881 -11.843 -0.0211>, rot=<0.059 -0.7043 -0.0605 -0.7048>, scale=<1 1.0001 1>}
 cape {parent=attach_cape, type=Null, loc=<0 -9.5929 0>, rot=<0 0 0 1>}
 attach_backpack {parent=spine_05, loc=<-0.0228 -13.5665 -0.0299>, rot=<-0.0119 -0.7068 0.0103 -0.7072>, scale=<1 1.0001 1>}
 Backpack_BR {parent=attach_backpack, type=Null, loc=<0 -9.0633 0>, rot=<-0.0551 0 0 0.9985>}
 clavicle_l {parent=spine_05, loc=<4.3575 -0.1466 -2.3223>, rot=<0.0465 -0.7289 -0.009 -0.6829>}
 upperarm_l {parent=clavicle_l, loc=<15.5111 0 0>, rot=<0.0226 -0.3234 -0.0088 -0.9459>}
 lowerarm_l {parent=upperarm_l, loc=<26.6514 0 0>, rot=<0.0391 -0.0099 -0.2453 -0.9686>}
 hand_l {parent=lowerarm_l, loc=<22.3491 0 0>, rot=<-0.7081 -0.0507 0.0975 -0.6975>}
 thumb_01_l {parent=hand_l, loc=<1.3118 -0.3191 -2.459>, rot=<0.6546 -0.4042 0.0664 -0.6354>}
 thumb_02_l {parent=thumb_01_l, loc=<5.9336 -0.0448 0.0401>, rot=<-0.0037 0.0213 0.298 -0.9543>}
 thumb_03_l {parent=thumb_02_l, loc=<2.2223 0.3682 -0.0907>, rot=<0.0118 -0.013 0.0066 -0.9998>}
 index_metacarpal_l {parent=hand_l, loc=<1.6535 0.0183 -2.96>, rot=<0.0158 -0.0124 -0.0173 -0.9996>}
 index_01_l {parent=index_metacarpal_l, loc=<8.3008 0.0357 -0.0524>, rot=<-0.0044 0.0485 0.0924 -0.9945>}
 index_02_l {parent=index_01_l, loc=<3.3587 -0.3004 0.0452>, rot=<-0.002 0.0027 0.3225 -0.9466>}
 index_03_l {parent=index_02_l, loc=<2.5578 -0.123 -0.0252>, rot=<-0.002 0.0067 0.2558 -0.9667>}
 middle_metacarpal_l {parent=hand_l, loc=<2.1856 -0.1725 -0.5984>, rot=<0.006 0.0168 -0.0329 -0.9993>}
 middle_01_l {parent=middle_metacarpal_l, loc=<7.3275 -0.2447 0.0129>, rot=<-0.0208 0.0042 0.0858 -0.9961>}
 middle_02_l {parent=middle_01_l, loc=<4.0123 0.1401 0.1175>, rot=<-0.0218 0.0258 0.3512 -0.9357>}
 middle_03_l {parent=middle_02_l, loc=<2.6077 0.0534 0.1373>, rot=<0.0082 0.0188 0.1708 -0.9851>}
 ring_metacarpal_l {parent=hand_l, loc=<2.2074 -0.3625 1.3638>, rot=<-0.0284 0.0722 -0.0458 -0.9959>}
 ring_01_l {parent=ring_metacarpal_l, loc=<7.0295 -0.6197 -0.1446>, rot=<-0.0489 -0.0252 0.1892 -0.9804>}
 ring_02_l {parent=ring_01_l, loc=<3.6174 0.4647 -0.1027>, rot=<0 0 0.2806 -0.9598>}
 ring_03_l {parent=ring_02_l, loc=<2.4092 0.0926 0.02>, rot=<0 0 0.2654 -0.9641>}
 pinky_metacarpal_l {parent=hand_l, loc=<1.7616 -0.9753 3.3814>, rot=<-0.0397 0.0814 -0.069 -0.9935>}
 pinky_01_l {parent=pinky_metacarpal_l, loc=<7.1709 -0.4831 -0.1094>, rot=<-0.0654 -0.0035 0.2242 -0.9723>}
 pinky_02_l {parent=pinky_01_l, loc=<2.8274 0.0893 -0.1043>, rot=<0.0037 0.0011 0.2854 -0.9584>}
 pinky_03_l {parent=pinky_02_l, loc=<2.2046 0.0307 -0.0078>, rot=<-0.0002 0.0098 0.1445 -0.9895>}
 deform_wrist_dn_base_l {parent=hand_l, loc=<0 -2.2342 0>, rot=<0 0 0 -1>}
 deform_wrist_dn_l {parent=deform_wrist_dn_base_l, loc=<0 0 0>, rot=<0 0 0 1>}
 deform_wrist_up_base_l {parent=hand_l, loc=<0 2.2682 0>, rot=<0 0 0 -1>}
 deform_wrist_up_l {parent=deform_wrist_up_base_l, loc=<0 0 0>, rot=<0 0 0 1>}
 lowerarm_twist_02_l {parent=lowerarm_l, loc=<9.6199 0 0>, rot=<0 0 0 -1>}
 lowerarm_twist_01_l {parent=lowerarm_l, loc=<17.8549 0 0>, rot=<0 0 0 -1>}
 deform_elbow_in_base_l {parent=lowerarm_l, loc=<0.2112 5.259 0.8274>, rot=<0 0 0 1>}
 deform_elbow_in_l {parent=deform_elbow_in_base_l, loc=<0 0 0>, rot=<0 0 0 1>}
 deform_elbow_out_base_l {parent=lowerarm_l, loc=<0 -4 0>, rot=<0 0 0 1>}
 deform_elbow_out_l {parent=deform_elbow_out_base_l, loc=<0 0 0>, rot=<0 0 0 1>}
 hand_l_null {parent=lowerarm_l, type=Null, loc=<22.3491 0 0>, rot=<-0.7081 -0.0507 0.0975 -0.6975>}
 upperarm_twist_01_l {parent=upperarm_l, loc=<0.0041 0 0>, rot=<0 0 0 -1>}
 deform_upperarm_bk_base_l {parent=upperarm_twist_01_l, loc=<1.0813 -7.2193 -1.9453>, rot=<0 0 0 -1>}
 deform_upperarm_bk_l {parent=deform_upperarm_bk_base_l, loc=<0 0 0>, rot=<0 0 0 1>}
 deform_upperarm_fr_base_l {parent=upperarm_twist_01_l, loc=<-0.3691 7.0506 -2.765>, rot=<0 0 0 -1>}
 deform_upperarm_fr_l {parent=deform_upperarm_fr_base_l, loc=<0 0 0>, rot=<0 0 0 1>}
 deform_upperarm_up_base_l {parent=upperarm_twist_01_l, loc=<-3.2216 -1.1391 3.3311>, rot=<0 0 0 1>}
 deform_upperarm_up_l {parent=deform_upperarm_up_base_l, loc=<0 0 0>, rot=<0 0 0 1>}
 upperarm_twist_02_l {parent=upperarm_l, loc=<13.8233 0.0387 0.1057>, rot=<-0.222 0.0032 -0.003 -0.975>}
 deform_upperarm_l {parent=clavicle_l, loc=<15.5111 -0.0001 0>, rot=<0.0226 -0.3234 -0.0088 -0.9459>}
 deform_lowerarm_l {parent=deform_upperarm_l, loc=<26.6514 0 0>, rot=<0.0391 -0.0099 -0.2453 -0.9686>}
 deform_hand_l {parent=deform_lowerarm_l, loc=<22.3491 0 0>, rot=<-0.7081 -0.0507 0.0975 -0.6975>}
 deform_lowerarm_twist_02_l {parent=deform_lowerarm_l, loc=<9.6199 0 0>, rot=<0 0 0 -1>}
 deform_lowerarm_twist_01_l {parent=deform_lowerarm_l, loc=<17.8549 0 0>, rot=<0 0 0 -1>}
 base_sleeve_out_l {parent=deform_lowerarm_twist_01_l, loc=<5.7352 -7.4487 -1.1679>, rot=<0.0052 -0.7423 0.117 -0.6597>}
 dyn_sleeve_out_l {parent=base_sleeve_out_l, loc=<0 0 0>, rot=<0 0 0 1>}
 base_sleeve_in_l {parent=deform_lowerarm_twist_01_l, loc=<4.9027 6.2368 -1.8521>, rot=<0.0686 -0.7501 0.045 -0.6562>}
 dyn_sleeve_in_l {parent=base_sleeve_in_l, loc=<0 0 0>, rot=<0 0 0 1>}
 deform_deform_elbow_out_base_l {parent=deform_lowerarm_l, loc=<0 -4 0>, rot=<0 0 0 -1>}
 deform_deform_elbow_out_l {parent=deform_deform_elbow_out_base_l, loc=<0 0 0>, rot=<0 0 0 1>}
 deform_deform_elbow_in_base_l {parent=deform_lowerarm_l, loc=<0.2112 5.2591 0.8274>, rot=<0 0 0 -1>}
 deform_deform_elbow_in_l {parent=deform_deform_elbow_in_base_l, loc=<0 0 0>, rot=<0 0 0 1>}
 deform_upperarm_twist_01_l {parent=deform_upperarm_l, loc=<0.0041 0 0>, rot=<0 0 0 1>}
 deform_deform_upperarm_bk_base_l {parent=deform_upperarm_twist_01_l, loc=<0.9962 -7.2716 -1.9414>, rot=<0.0001 0 0 -1>}
 deform_deform_upperarm_bk_l {parent=deform_deform_upperarm_bk_base_l, loc=<0 0 0>, rot=<0 0 0 1>}
 deform_deform_upperarm_fr_base_l {parent=deform_upperarm_twist_01_l, loc=<-0.33 7.0764 -2.7569>, rot=<0.0001 0 0 -1>}
 deform_deform_upperarm_fr_l {parent=deform_deform_upperarm_fr_base_l, loc=<0 0 0>, rot=<0 0 0 1>}
 deform_deform_upperarm_up_base_l {parent=deform_upperarm_twist_01_l, loc=<-3.2216 -1.1391 3.3311>, rot=<0 0 0 1>}
 deform_deform_upperarm_up_l {parent=deform_deform_upperarm_up_base_l, loc=<0 0 0>, rot=<0 0 0 1>}
 deform_upperarm_twist_02_l {parent=deform_upperarm_l, loc=<13.7941 0 0>, rot=<0 0 0 1>}
 clavicle_r {parent=spine_05, loc=<4.3575 -0.1466 2.3223>, rot=<0.7289 0.0465 -0.6829 0.009>}
 upperarm_r {parent=clavicle_r, loc=<-15.5111 0 0>, rot=<0.0226 -0.3234 -0.0088 -0.9459>}
 lowerarm_r {parent=upperarm_r, loc=<-26.6514 0 0>, rot=<0.0391 -0.0099 -0.2453 -0.9686>}
 hand_r {parent=lowerarm_r, loc=<-22.3491 0 0>, rot=<-0.7081 -0.0507 0.0975 -0.6975>}
 thumb_01_r {parent=hand_r, loc=<-1.3118 0.3191 2.459>, rot=<0.6546 -0.4042 0.0664 -0.6354>}
 thumb_02_r {parent=thumb_01_r, loc=<-5.9336 0.0448 -0.0401>, rot=<-0.0037 0.0213 0.298 -0.9543>}
 thumb_03_r {parent=thumb_02_r, loc=<-2.2223 -0.3682 0.0907>, rot=<0.0118 -0.013 0.0066 -0.9998>}
 index_metacarpal_r {parent=hand_r, loc=<-1.6535 -0.0183 2.96>, rot=<0.0158 -0.0124 -0.0173 -0.9996>}
 index_01_r {parent=index_metacarpal_r, loc=<-8.3008 -0.0357 0.0524>, rot=<-0.0044 0.0485 0.0924 -0.9945>}
 index_02_r {parent=index_01_r, loc=<-3.3587 0.3004 -0.0452>, rot=<-0.002 0.0027 0.3225 -0.9466>}
 index_03_r {parent=index_02_r, loc=<-2.5578 0.123 0.0252>, rot=<-0.002 0.0067 0.2558 -0.9667>}
 middle_metacarpal_r {parent=hand_r, loc=<-2.1856 0.1725 0.5984>, rot=<0.006 0.0168 -0.0329 -0.9993>}
 middle_01_r {parent=middle_metacarpal_r, loc=<-7.3275 0.2447 -0.0129>, rot=<-0.0208 0.0042 0.0858 -0.9961>}
 middle_02_r {parent=middle_01_r, loc=<-4.0123 -0.1401 -0.1175>, rot=<-0.0218 0.0258 0.3512 -0.9357>}
 middle_03_r {parent=middle_02_r, loc=<-2.6077 -0.0534 -0.1373>, rot=<0.0082 0.0188 0.1708 -0.9851>}
 ring_metacarpal_r {parent=hand_r, loc=<-2.2074 0.3625 -1.3638>, rot=<-0.0284 0.0722 -0.0458 -0.9959>}
 ring_01_r {parent=ring_metacarpal_r, loc=<-7.0295 0.6197 0.1446>, rot=<-0.0489 -0.0252 0.1892 -0.9804>}
 ring_02_r {parent=ring_01_r, loc=<-3.6174 -0.4647 0.1027>, rot=<0 0 0.2806 -0.9598>}
 ring_03_r {parent=ring_02_r, loc=<-2.4305 -0.1306 -0.02>, rot=<0 0 0.5209 -0.8536>}
 pinky_metacarpal_r {parent=hand_r, loc=<-1.7616 0.9753 -3.3814>, rot=<-0.0397 0.0814 -0.069 -0.9935>}
 pinky_01_r {parent=pinky_metacarpal_r, loc=<-7.1709 0.4831 0.1094>, rot=<-0.0654 -0.0035 0.2242 -0.9723>}
 pinky_02_r {parent=pinky_01_r, loc=<-2.8274 -0.0893 0.1043>, rot=<0.0037 0.0011 0.2854 -0.9584>}
 pinky_03_r {parent=pinky_02_r, loc=<-2.2046 -0.0307 0.0078>, rot=<-0.0002 0.0098 0.1445 -0.9895>}
 deform_wrist_dn_base_r {parent=hand_r, loc=<-0.0004 2.2345 -0.0001>, rot=<0 0 0 -1>}
 deform_wrist_dn_r {parent=deform_wrist_dn_base_r, loc=<0 0 0>, rot=<0 0 0 1>}
 deform_wrist_up_base_r {parent=hand_r, loc=<0.0003 -2.2684 0.0001>, rot=<0 0 0 -1>}
 deform_wrist_up_r {parent=deform_wrist_up_base_r, loc=<0 0 0>, rot=<0 0 0 1>}
 lowerarm_twist_02_r {parent=lowerarm_r, loc=<-9.619 0.0963 0>, rot=<-0.255 0.0013 0.0048 -0.9669>}
 lowerarm_twist_01_r {parent=lowerarm_r, loc=<-17.8549 0 0>, rot=<0 0 0 -1>}
 deform_elbow_in_base_r {parent=lowerarm_r, loc=<-0.2111 -5.2591 -0.8275>, rot=<0 0 0 -1>}
 deform_elbow_in_r {parent=deform_elbow_in_base_r, loc=<-0.0001 0 0.0002>, rot=<0 0 0 1>}
 deform_elbow_out_base_r {parent=lowerarm_r, loc=<0 4 -0.0001>, rot=<0 0 0 -1>}
 deform_elbow_out_r {parent=deform_elbow_out_base_r, loc=<0 0 0>, rot=<0 0 0 1>}
 hand_r_null {parent=lowerarm_r, type=Null, loc=<-22.3491 0 0>, rot=<-0.7081 -0.0507 0.0975 -0.6975>}
 upperarm_twist_01_r {parent=upperarm_r, loc=<-0.0041 0 0>, rot=<0 0 0 -1>}
 deform_upperarm_bk_base_r {parent=upperarm_twist_01_r, loc=<-1.0814 7.2192 1.9455>, rot=<0 0 0 -1>}
 deform_upperarm_bk_r {parent=deform_upperarm_bk_base_r, loc=<0 0 0>, rot=<0 0 0 1>}
 deform_upperarm_fr_base_r {parent=upperarm_twist_01_r, loc=<0.3689 -7.0506 2.7652>, rot=<0 0 0 -1>}
 deform_upperarm_fr_r {parent=deform_upperarm_fr_base_r, loc=<0 0 0>, rot=<0 0 0 1>}
 deform_upperarm_up_base_r {parent=upperarm_twist_01_r, loc=<3.2214 1.1391 -3.3308>, rot=<0 0 0 1>}
 deform_upperarm_up_r {parent=deform_upperarm_up_base_r, loc=<0 0 0>, rot=<0 0 0 -1>}
 upperarm_twist_02_r {parent=upperarm_r, loc=<-13.73 -0.0567 0.0507>, rot=<-0.1577 -0.0022 -0.0015 -0.9875>}
 deform_upperarm_r {parent=clavicle_r, loc=<-15.5111 0 0>, rot=<0.0226 -0.3234 -0.0088 -0.9459>}
 deform_lowerarm_r {parent=deform_upperarm_r, loc=<-26.6514 0 0>, rot=<0.0391 -0.0099 -0.2453 -0.9686>}
 deform_hand_r {parent=deform_lowerarm_r, loc=<-22.3491 0 0>, rot=<-0.7081 -0.0507 0.0975 -0.6975>}
 deform_lowerarm_twist_02_r {parent=deform_lowerarm_r, loc=<-9.6199 0 0>, rot=<0 0 0 -1>}
 deform_lowerarm_twist_01_r {parent=deform_lowerarm_r, loc=<-17.8549 0 0>, rot=<0 0 0 -1>}
 base_sleeve_in_r {parent=deform_lowerarm_twist_01_r, loc=<-4.9024 -6.2369 1.8518>, rot=<0.045 0.6562 -0.0686 -0.7501>}
 dyn_sleeve_in_r {parent=base_sleeve_in_r, loc=<0 0 0>, rot=<0 0 0 1>}
 base_sleeve_out_r {parent=deform_lowerarm_twist_01_r, loc=<-5.7352 7.4487 1.1678>, rot=<0.117 0.6597 -0.0052 -0.7423>}
 dyn_sleeve_out_r {parent=base_sleeve_out_r, loc=<0 0 0>, rot=<0 0 0 1>}
 deform_deform_elbow_in_base_r {parent=deform_lowerarm_r, loc=<-0.2111 -5.2591 -0.8275>, rot=<0 0 0 -1>}
 deform_deform_elbow_in_r {parent=deform_deform_elbow_in_base_r, loc=<-0.0001 0 0.0002>, rot=<0 0 0 1>}
 deform_deform_elbow_out_base_r {parent=deform_lowerarm_r, loc=<0 4 -0.0001>, rot=<0 0 0 -1>}
 deform_deform_elbow_out_r {parent=deform_deform_elbow_out_base_r, loc=<0 0 0>, rot=<0 0 0 1>}
 deform_upperarm_twist_01_r {parent=deform_upperarm_r, loc=<-0.0041 0 0>, rot=<0.1515 -0.0015 -0.0023 -0.9885>}
 deform_deform_upperarm_bk_base_r {parent=deform_upperarm_twist_01_r, loc=<-1.0245 7.271 1.9606>, rot=<-0.0008 0 0 -1>}
 deform_deform_upperarm_bk_r {parent=deform_deform_upperarm_bk_base_r, loc=<0 0 0>, rot=<0 0 0 1>}
 deform_deform_upperarm_fr_base_r {parent=deform_upperarm_twist_01_r, loc=<0.3093 -7.1362 2.7543>, rot=<-0.0008 0 0 -1>}
 deform_deform_upperarm_fr_r {parent=deform_deform_upperarm_fr_base_r, loc=<0 0 0>, rot=<0 0 0 1>}
 deform_deform_upperarm_up_base_r {parent=deform_upperarm_twist_01_r, loc=<3.2202 1.1746 -3.3315>, rot=<-0.0008 0 0 -1>}
 deform_deform_upperarm_up_r {parent=deform_deform_upperarm_up_base_r, loc=<0 0 0>, rot=<0 0 0 -1>}
 deform_upperarm_twist_02_r {parent=deform_upperarm_r, loc=<-13.7941 0 0>, rot=<0 0 0 -1>}
 base_bow_m {parent=spine_05, loc=<1.4594 19.9934 0>, rot=<-0.0284 0.7065 0.7065 -0.0284>}
 dyn_bow_m {parent=base_bow_m, loc=<0 0 0>, rot=<0 0 0 1>}
 dyn_bowWing_r {parent=dyn_bow_m, loc=<-0.2475 1.7575 0.5688>, rot=<0 0 0 1>}
 dyn_bowWing_l {parent=dyn_bow_m, loc=<-0.2475 -1.7575 0.5688>, rot=<0 0 0 1>}
 dyn_coatLapel_l {parent=spine_04, loc=<4.5233 12.9865 -16.1458>, rot=<0.0813 -0.0212 0.976 0.201>}
 dyn_coatLapel_r {parent=spine_04, loc=<4.5233 12.9865 16.1458>, rot=<0.0212 0.0813 0.201 -0.976>}
 base_tube_01_m {parent=spine_03, loc=<5.9861 21.2683 2.0644>, rot=<0.1137 0.6832 0.72 -0.0436>}
 dyn_tube_01_m {parent=base_tube_01_m, loc=<0 0 0>, rot=<0 0 0 1>}
 base_tube_02_m {parent=spine_03, loc=<0.6829 21.3091 -4.826>, rot=<-0.1446 -0.7527 -0.6422 0.0093>}
 dyn_tube_02_m {parent=base_tube_02_m, loc=<0 0 0>, rot=<0 0 0 1>}
 base_tube_03_m {parent=spine_03, loc=<-4.011 20.1442 -10.7776>, rot=<-0.1334 -0.8103 -0.5705 0.0071>}
 dyn_tube_03_m {parent=base_tube_03_m, loc=<0 0 0>, rot=<0 0 0 1>}
 thigh_l {parent=pelvis, loc=<-2.5693 -0.0109 -7.7501>, rot=<0.0012 0.0572 0.0155 -0.9982>}
 calf_l {parent=thigh_l, loc=<-39.0484 0 0>, rot=<0.0006 -0.0161 -0.0391 -0.9991>}
 foot_l {parent=calf_l, loc=<-42.561 0.1077 0.4292>, rot=<0.0418 -0.0455 0.0343 -0.9975>}
 ball_l {parent=foot_l, loc=<-6.1599 13.8264 0.6806>, rot=<0.0001 -0.0001 -0.7063 -0.7079>}
 calf_twist_02_l {parent=calf_l, loc=<-18.8064 -0.0967 0.0397>, rot=<0.0237 0.0095 -0.0035 -0.9997>}
 calf_twist_01_l {parent=calf_l, loc=<-35.813 -0.0967 0.0397>, rot=<0.0237 0.0095 -0.0035 -0.9997>}
 deform_knee_in_base_l {parent=calf_l, loc=<1.298 -5.2957 -0.0496>, rot=<0 0 0 -1>}
 deform_knee_in_l {parent=deform_knee_in_base_l, loc=<0 0 0>, rot=<0 0 0 1>}
 deform_knee_out_base_l {parent=calf_l, loc=<0 4 0>, rot=<0 0 0 -1>}
 deform_knee_out_l {parent=deform_knee_out_base_l, loc=<0 0 0>, rot=<0 0 0 1>}
 thigh_twist_01_l {parent=thigh_l, loc=<-21.6303 0 0>, rot=<0.0013 0 0 -1>}
 deform_groin_base_l {parent=thigh_l, loc=<-2.1908 8.1142 -1.7442>, rot=<0 0 0 -1>}
 deform_groin_l {parent=deform_groin_base_l, loc=<0 0 0>, rot=<0 0 0 1>}
 thigh_r {parent=pelvis, loc=<-2.5693 -0.0109 7.7501>, rot=<0.0572 -0.0012 0.9982 0.0155>}
 calf_r {parent=thigh_r, loc=<39.0484 0 0>, rot=<0.0006 -0.0161 -0.0391 -0.9991>}
 foot_r {parent=calf_r, loc=<42.561 -0.1077 -0.4292>, rot=<0.0418 -0.0455 0.0343 -0.9975>}
 ball_r {parent=foot_r, loc=<6.1599 -13.8264 -0.6806>, rot=<0.0001 -0.0001 -0.7063 -0.7079>}
 calf_twist_02_r {parent=calf_r, loc=<18.8064 0.0967 -0.0397>, rot=<0.0237 0.0095 -0.0035 -0.9997>}
 calf_twist_01_r {parent=calf_r, loc=<35.813 0.0967 -0.0397>, rot=<0.0237 0.0095 -0.0035 -0.9997>}
 deform_knee_in_base_r {parent=calf_r, loc=<-1.298 5.2957 0.0496>, rot=<0 0 0 -1>}
 deform_knee_in_r {parent=deform_knee_in_base_r, loc=<0 0 0>, rot=<0 0 0 1>}
 deform_knee_out_base_r {parent=calf_r, loc=<0 -4 0>, rot=<0 0 0 -1>}
 deform_knee_out_r {parent=deform_knee_out_base_r, loc=<0 0 0>, rot=<0 0 0 1>}
 thigh_twist_01_r {parent=thigh_r, loc=<21.6303 0 0>, rot=<0.0013 0 0 -1>}
 deform_groin_base_r {parent=thigh_r, loc=<2.1908 -8.1142 1.7442>, rot=<0 0 0 -1>}
 deform_groin_r {parent=deform_groin_base_r, loc=<0 0 0>, rot=<0 0 0 1>}
 deform_glute_base_l {parent=pelvis, loc=<-8.8064 -8.8222 -8.4287>, rot=<0.0275 0.0501 -0.4472 -0.8926>}
 deform_glute_l {parent=deform_glute_base_l, loc=<0 0 0>, rot=<0 0 0 1>}
 deform_glute_base_r {parent=pelvis, loc=<-8.8062 -8.8222 8.4287>, rot=<0.0501 -0.0275 0.8926 -0.4472>}
 deform_glute_r {parent=deform_glute_base_r, loc=<0 0 0>, rot=<0 0 0 1>}
 base_cloth_l_fr {parent=pelvis, loc=<2.9972 15.083 -18.1399>, rot=<0.0204 -0.9223 -0.3819 -0.0547>}
 dyn_cloth_01_l_fr {parent=base_cloth_l_fr, loc=<0 0 0>, rot=<0 0 0 1>}
 dyn_cloth_02_l_fr {parent=dyn_cloth_01_l_fr, loc=<14.3311 0 0>, rot=<0.0009 -0.0184 -0.0058 -0.9998>}
 dyn_cloth_03_l_fr {parent=dyn_cloth_02_l_fr, loc=<14.8438 0 0>, rot=<0 0 0 1>}
 base_cloth_l {parent=pelvis, loc=<2.1213 -1.4378 -21.7642>, rot=<-0.0184 -0.9986 0.0009 -0.0506>}
 dyn_cloth_01_l {parent=base_cloth_l, loc=<0 0 0>, rot=<0 0 0 1>}
 dyn_cloth_02_l {parent=dyn_cloth_01_l, loc=<14.509 0 0>, rot=<0 -0.0222 0.0002 -0.9998>}
 dyn_cloth_03_l {parent=dyn_cloth_02_l, loc=<12.6873 0 0>, rot=<0 0 0 1>}
 base_cloth_l_bck {parent=pelvis, loc=<2.0352 -18.5374 -14.9937>, rot=<-0.0259 -0.8655 0.5001 -0.0083>}
 dyn_cloth_01_l_bck {parent=base_cloth_l_bck, loc=<0 0 0>, rot=<0 0 0 1>}
 dyn_cloth_02_l_bck {parent=dyn_cloth_01_l_bck, loc=<14.3599 0 0>, rot=<0 -0.0002 0 -1>}
 dyn_cloth_03_l_bck {parent=dyn_cloth_02_l_bck, loc=<12.7073 0 0>, rot=<0 0 0 1>}
 base_cloth_bck {parent=pelvis, loc=<1.884 -23.2023 0.2721>, rot=<-0.0127 -0.707 0.707 0.0127>}
 dyn_cloth_01_bck {parent=base_cloth_bck, loc=<0 0 0>, rot=<0 0 0 1>}
 dyn_cloth_02_bck {parent=dyn_cloth_01_bck, loc=<14.4068 0 0>, rot=<0 0 0 1>}
 dyn_cloth_03_bck {parent=dyn_cloth_02_bck, loc=<12.0476 0 0>, rot=<0 0 0 1>}
 base_cloth_r {parent=pelvis, loc=<2.1213 -1.4378 21.7642>, rot=<0.0506 -0.0009 0.9986 0.0184>}
 dyn_cloth_01_r {parent=base_cloth_r, loc=<0 0 0>, rot=<0 0 0 1>}
 dyn_cloth_02_r {parent=dyn_cloth_01_r, loc=<14.509 0 0>, rot=<0 -0.0222 -0.0002 -0.9998>}
 dyn_cloth_03_r {parent=dyn_cloth_02_r, loc=<12.6873 0 0.0001>, rot=<0 0 0 1>}
 base_cloth_r_bck {parent=pelvis, loc=<2.0352 -18.5374 14.9937>, rot=<0.0083 -0.5001 0.8655 0.0259>}
 dyn_cloth_01_r_bck {parent=base_cloth_r_bck, loc=<0 0 0>, rot=<0 0 0 1>}
 dyn_cloth_02_r_bck {parent=dyn_cloth_01_r_bck, loc=<14.3599 0 0.0001>, rot=<0 -0.0002 0 -1>}
 dyn_cloth_03_r_bck {parent=dyn_cloth_02_r_bck, loc=<12.7074 0 -0.0001>, rot=<0 0 0 1>}
 base_cloth_r_fr {parent=pelvis, loc=<2.9578 15.4337 17.7918>, rot=<0.0547 0.3819 0.9223 -0.0204>}
 dyn_cloth_01_r_fr {parent=base_cloth_r_fr, loc=<0 0 0>, rot=<0 0 0 1>}
 dyn_cloth_02_r_fr {parent=dyn_cloth_01_r_fr, loc=<14.331 0 -0.0001>, rot=<-0.0009 -0.0184 0.0058 -0.9998>}
 dyn_cloth_03_r_fr {parent=dyn_cloth_02_r_fr, loc=<14.8439 0 0.0001>, rot=<0 0 0 1>}
 thigh_r_null {parent=pelvis, type=Null, loc=<-2.5693 -0.0109 7.7501>, rot=<0.0572 -0.0012 0.9982 0.0155>}
 thigh_l_null {parent=pelvis, type=Null, loc=<-2.5693 -0.0109 -7.7501>, rot=<0.0012 0.0572 0.0155 -0.9982>}
]
variables [
 guanteletEquipped {type=double, default=0}
 isCrouchWalkin {type=double, default=0}
 isCrouching {type=bool, default=False}
 isDBNO {type=bool, default=False}
 isEmoting {type=bool, default=False}
 isKneeSlide {type=bool, default=False}
 isManequin {type=bool, default=False}
 isSliding {type=bool, default=False}
 rigidbodyenable {type=bool, default=False}
 thigh_L_Back {type=double, default=0}
 thigh_R_Back {type=double, default=0}
]
