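# Physics asset for the knee-sliding variant of the DairyString WadePuck male body, covering seven three-segment cloth bone chains around the lower body.
# Limited swing and twist constraints link chain segments, but only the centre-back chain's first segment is limited to its base bone.
# Other first segments, base bones, pelvis and legs tie to root with free angular motion.
# Pelvis and leg shapes accompany a huge capsule offset below root.
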PhysicsAsset /BRCosmetics/Characters/Player/Male/Medium/Bodies/M_MED_DairyString_WadePuck/Meshes/M_MED_DairyString_WadePuck_KneeSliding_Physics
bodies {
 base_cloth_bck [{shapeType=Capsule, length=0.505, radius=0.505, shapeName=base_cloth_bck_capsule, rotation=<rot 0 -90 0>}]
 base_cloth_l [{shapeType=Capsule, length=0.505, radius=0.505, shapeName=base_cloth_l_capsule, rotation=<rot 0 -90 0>}]
 base_cloth_l_bck [{shapeType=Capsule, length=0.505, radius=0.505, shapeName=base_cloth_l_bck_capsule, rotation=<rot 0 -90 0>}]
 base_cloth_l_fr [{shapeType=Capsule, length=0.505, radius=0.505, shapeName=base_cloth_l_fr_capsule, rotation=<rot 0 -90 0>}]
 base_cloth_r [{shapeType=Capsule, length=0.505, radius=0.505, shapeName=base_cloth_r_capsule, rotation=<rot 0 -90 0>}]
 base_cloth_r_bck [{shapeType=Capsule, length=0.505, radius=0.505, shapeName=base_cloth_r_bck_capsule, rotation=<rot 0 -90 0>}]
 base_cloth_r_fr [{shapeType=Capsule, length=0.505, radius=0.505, shapeName=base_cloth_r_fr_capsule, rotation=<rot 0 -90 0>}]
 calf_l [{shapeType=Capsule, length=37.0119, radius=10.5, center=<-21.1142 0 0>, rotation=<rot 90 0 0>}]
 calf_r [{shapeType=Capsule, length=37.0119, radius=10.5, center=<21.1142 0 0>, rotation=<rot 90 180 0>}]
 dyn_cloth_01_bck [{shapeType=Capsule, length=10, radius=10, shapeName=dyn_cloth_01_bck_capsule, center=<10.3442 0 6.7868>, rotation=<rot 0 0 90>}]
 dyn_cloth_01_l [{shapeType=Capsule, length=14.6111, radius=5.3889, shapeName=dyn_cloth_01_l_capsule, center=<10.3442 0 6.7868>, rotation=<rot 0 0 90>}]
 dyn_cloth_01_l_bck [{shapeType=Capsule, length=14.6111, radius=5.3889, shapeName=dyn_cloth_01_l_bck_capsule, center=<10.3442 0 6.7868>, rotation=<rot 0 0 90>}]
 dyn_cloth_01_l_fr [{shapeType=Capsule, length=14.6111, radius=5.3889, shapeName=dyn_cloth_01_l_fr_capsule, center=<10.3442 0 6.7868>, rotation=<rot 0.0002 0 90>}]
 dyn_cloth_01_r [{shapeType=Capsule, length=14.6111, radius=5.3889, shapeName=dyn_cloth_01_r_capsule, center=<10.3442 0 6.7868>, rotation=<rot 0 0 90>}]
 dyn_cloth_01_r_bck [{shapeType=Capsule, length=14.6111, radius=5.3889, shapeName=dyn_cloth_01_r_bck_capsule, center=<10.3442 0 6.7868>, rotation=<rot 0 0 90>}]
 dyn_cloth_01_r_fr [{shapeType=Capsule, length=14.6111, radius=5.3889, shapeName=dyn_cloth_01_r_fr_capsule, center=<10.3442 0 8.3404>, rotation=<rot 0 0 90>}]
 dyn_cloth_02_bck [{shapeType=Capsule, length=19.0875, radius=10, shapeName=dyn_cloth_02_bck_capsule, center=<10.3442 0 6.7868>, rotation=<rot 0 0 90>}]
 dyn_cloth_02_l [{shapeType=Capsule, length=10, radius=10, shapeName=dyn_cloth_02_l_capsule, center=<10.3442 0 7.9255>, rotation=<rot 0 0 90>}]
 dyn_cloth_02_l_bck [{shapeType=Capsule, length=19.0875, radius=10, shapeName=dyn_cloth_02_l_bck_capsule, center=<10.3442 0 6.7868>, rotation=<rot 0 0 90>}]
 dyn_cloth_02_l_fr [{shapeType=Capsule, length=10, radius=10, shapeName=dyn_cloth_02_l_fr_capsule, center=<10.3442 0 6.7868>, rotation=<rot 0.0002 0 90>}]
 dyn_cloth_02_r [{shapeType=Capsule, length=10, radius=10, shapeName=dyn_cloth_02_r_capsule, center=<10.3442 0 8.3528>, rotation=<rot 0 0 90>}]
 dyn_cloth_02_r_bck [{shapeType=Capsule, length=10, radius=10, shapeName=dyn_cloth_02_r_bck_capsule, center=<10.3442 0 6.7868>, rotation=<rot 0 0 90>}]
 dyn_cloth_02_r_fr [{shapeType=Capsule, length=10, radius=10, shapeName=dyn_cloth_02_r_fr_capsule, center=<10.3442 0 8.3404>, rotation=<rot 0 0 90>}]
 dyn_cloth_03_bck [{shapeType=Capsule, length=21.788, radius=13.001, shapeName=dyn_cloth_03_bck_capsule, center=<10.3442 0 8.0811>, rotation=<rot 0 0 90>}]
 dyn_cloth_03_l [{shapeType=Capsule, length=10, radius=10, shapeName=dyn_cloth_03_l_capsule, center=<10.3442 0 8.3478>, rotation=<rot 0 0 90>}]
 dyn_cloth_03_l_bck [{shapeType=Capsule, length=21.788, radius=13.001, shapeName=dyn_cloth_03_l_bck_capsule, center=<10.3442 0 8.0811>, rotation=<rot 0 0 90>}]
 dyn_cloth_03_l_fr [{shapeType=Capsule, length=10, radius=10, shapeName=dyn_cloth_03_l_fr_capsule, center=<10.3442 0 7.091>, rotation=<rot 0.0002 0 90>}]
 dyn_cloth_03_r [{shapeType=Capsule, length=10, radius=10, shapeName=dyn_cloth_03_r_capsule, center=<10.3442 0 8.3528>, rotation=<rot 0 0 90>}]
 dyn_cloth_03_r_bck [{shapeType=Capsule, length=21.788, radius=13.001, shapeName=dyn_cloth_03_r_bck_capsule, center=<10.3442 0 8.0811>, rotation=<rot 0 0 90>}]
 dyn_cloth_03_r_fr [{shapeType=Capsule, length=10, radius=10, shapeName=dyn_cloth_03_r_fr_capsule, center=<10.3442 0 9.7322>, rotation=<rot 0 0 90>}]
 foot_l [{shapeType=Capsule, length=22, radius=10, center=<-4.447 6.5429 0>, rotation=<rot 0 0 90>}]
 foot_r [{shapeType=Capsule, length=22, radius=10, center=<4.447 -6.5429 0>, rotation=<rot 0 0 -90>}]
 pelvis [{shapeType=Sphere, radius=21.4723, shapeName=pelvis_sphere, center=<-6.1021 0 1.7505>}, {shapeType=Capsule, length=1.6855, radius=21.1882}]
 root [{shapeType=Capsule, length=531.644, radius=213.6752, shapeName=root_capsule, center=<0 0 -350.5208>, rotation=<rot 0 0 90>}]
 thigh_l [{shapeType=Capsule, length=30.5132, radius=10.3144, shapeName=thigh_l_capsule, center=<-17.8618 0.2155 -0.2511>, rotation=<rot 90 0 0>}]
 thigh_r [{shapeType=Capsule, length=30.5132, radius=10.3144, shapeName=thigh_r_capsule, center=<17.8618 -0.2155 0.2511>, rotation=<rot 90 180 0>}]
}
constraints [
 {bone1Name=dyn_cloth_01_bck, bone2Name=base_cloth_bck, swing1LimitDegrees=0, swing2LimitDegrees=20, twistLimitDegrees=2}
 {bone1Name=dyn_cloth_02_bck, bone2Name=dyn_cloth_01_bck, swing1LimitDegrees=0, swing2LimitDegrees=30, twistLimitDegrees=2}
 {bone1Name=dyn_cloth_03_bck, bone2Name=dyn_cloth_02_bck, swing1LimitDegrees=0, swing2LimitDegrees=30, twistLimitDegrees=2}
 {bone1Name=dyn_cloth_02_l, bone2Name=dyn_cloth_01_l, swing1LimitDegrees=10, swing2LimitDegrees=15, twistLimitDegrees=0}
 {bone1Name=dyn_cloth_03_l, bone2Name=dyn_cloth_02_l, swing1LimitDegrees=10, swing2LimitDegrees=15, twistLimitDegrees=2}
 {bone1Name=dyn_cloth_02_l_bck, bone2Name=dyn_cloth_01_l_bck, swing1LimitDegrees=0, swing2LimitDegrees=30, twistLimitDegrees=0}
 {bone1Name=dyn_cloth_03_l_bck, bone2Name=dyn_cloth_02_l_bck, swing1LimitDegrees=10, swing2LimitDegrees=30, twistLimitDegrees=2}
 {bone1Name=dyn_cloth_02_l_fr, bone2Name=dyn_cloth_01_l_fr, swing1LimitDegrees=5, swing2LimitDegrees=25, twistLimitDegrees=2}
 {bone1Name=dyn_cloth_03_l_fr, bone2Name=dyn_cloth_02_l_fr, swing1LimitDegrees=5, swing2LimitDegrees=15, twistLimitDegrees=2}
 {bone1Name=dyn_cloth_02_r, bone2Name=dyn_cloth_01_r, swing1LimitDegrees=10, swing2LimitDegrees=10, twistLimitDegrees=0}
 {bone1Name=dyn_cloth_03_r, bone2Name=dyn_cloth_02_r, swing1LimitDegrees=10, swing2LimitDegrees=10, twistLimitDegrees=2}
 {bone1Name=dyn_cloth_02_r_bck, bone2Name=dyn_cloth_01_r_bck, swing1LimitDegrees=0, swing2LimitDegrees=30, twistLimitDegrees=0}
 {bone1Name=dyn_cloth_03_r_bck, bone2Name=dyn_cloth_02_r_bck, swing1LimitDegrees=10, swing2LimitDegrees=30, twistLimitDegrees=2}
 {bone1Name=dyn_cloth_02_r_fr, bone2Name=dyn_cloth_01_r_fr, swing1LimitDegrees=5, swing2LimitDegrees=15, twistLimitDegrees=0}
 {bone1Name=dyn_cloth_03_r_fr, bone2Name=dyn_cloth_02_r_fr, swing1LimitDegrees=5, swing2LimitDegrees=15, twistLimitDegrees=2}
 {
  bone1Name=pelvis
  bone2Name=root
  swing1LimitDegrees=45
  swing1Motion=Free
  swing2LimitDegrees=45
  swing2Motion=Free
  twistLimitDegrees=45
  twistMotion=Free
 }
 {
  bone1Name=thigh_l
  bone2Name=root
  swing1LimitDegrees=45
  swing1Motion=Free
  swing2LimitDegrees=45
  swing2Motion=Free
  twistLimitDegrees=45
  twistMotion=Free
 }
 {
  bone1Name=calf_l
  bone2Name=root
  swing1LimitDegrees=45
  swing1Motion=Free
  swing2LimitDegrees=45
  swing2Motion=Free
  twistLimitDegrees=45
  twistMotion=Free
 }
 {
  bone1Name=foot_l
  bone2Name=root
  swing1LimitDegrees=45
  swing1Motion=Free
  swing2LimitDegrees=45
  swing2Motion=Free
  twistLimitDegrees=45
  twistMotion=Free
 }
 {
  bone1Name=thigh_r
  bone2Name=root
  swing1LimitDegrees=45
  swing1Motion=Free
  swing2LimitDegrees=45
  swing2Motion=Free
  twistLimitDegrees=45
  twistMotion=Free
 }
 {
  bone1Name=calf_r
  bone2Name=root
  swing1LimitDegrees=45
  swing1Motion=Free
  swing2LimitDegrees=45
  swing2Motion=Free
  twistLimitDegrees=45
  twistMotion=Free
 }
 {
  bone1Name=foot_r
  bone2Name=root
  swing1LimitDegrees=45
  swing1Motion=Free
  swing2LimitDegrees=45
  swing2Motion=Free
  twistLimitDegrees=45
  twistMotion=Free
 }
 {
  bone1Name=base_cloth_l_fr
  bone2Name=root
  swing1LimitDegrees=45
  swing1Motion=Free
  swing2LimitDegrees=45
  swing2Motion=Free
  twistLimitDegrees=45
  twistMotion=Free
 }
 {
  bone1Name=dyn_cloth_01_l_fr
  bone2Name=root
  swing1LimitDegrees=45
  swing1Motion=Free
  swing2LimitDegrees=45
  swing2Motion=Free
  twistLimitDegrees=45
  twistMotion=Free
 }
 {
  bone1Name=base_cloth_l
  bone2Name=root
  swing1LimitDegrees=45
  swing1Motion=Free
  swing2LimitDegrees=45
  swing2Motion=Free
  twistLimitDegrees=45
  twistMotion=Free
 }
 {
  bone1Name=dyn_cloth_01_l
  bone2Name=root
  swing1LimitDegrees=45
  swing1Motion=Free
  swing2LimitDegrees=45
  swing2Motion=Free
  twistLimitDegrees=45
  twistMotion=Free
 }
 {
  bone1Name=base_cloth_l_bck
  bone2Name=root
  swing1LimitDegrees=45
  swing1Motion=Free
  swing2LimitDegrees=45
  swing2Motion=Free
  twistLimitDegrees=45
  twistMotion=Free
 }
 {
  bone1Name=dyn_cloth_01_l_bck
  bone2Name=root
  swing1LimitDegrees=45
  swing1Motion=Free
  swing2LimitDegrees=45
  swing2Motion=Free
  twistLimitDegrees=45
  twistMotion=Free
 }
 {
  bone1Name=base_cloth_bck
  bone2Name=root
  swing1LimitDegrees=45
  swing1Motion=Free
  swing2LimitDegrees=45
  swing2Motion=Free
  twistLimitDegrees=45
  twistMotion=Free
 }
 {
  bone1Name=base_cloth_r
  bone2Name=root
  swing1LimitDegrees=45
  swing1Motion=Free
  swing2LimitDegrees=45
  swing2Motion=Free
  twistLimitDegrees=45
  twistMotion=Free
 }
 {
  bone1Name=dyn_cloth_01_r
  bone2Name=root
  swing1LimitDegrees=45
  swing1Motion=Free
  swing2LimitDegrees=45
  swing2Motion=Free
  twistLimitDegrees=45
  twistMotion=Free
 }
 {
  bone1Name=base_cloth_r_bck
  bone2Name=root
  swing1LimitDegrees=45
  swing1Motion=Free
  swing2LimitDegrees=45
  swing2Motion=Free
  twistLimitDegrees=45
  twistMotion=Free
 }
 {
  bone1Name=dyn_cloth_01_r_bck
  bone2Name=root
  swing1LimitDegrees=45
  swing1Motion=Free
  swing2LimitDegrees=45
  swing2Motion=Free
  twistLimitDegrees=45
  twistMotion=Free
 }
 {
  bone1Name=base_cloth_r_fr
  bone2Name=root
  swing1LimitDegrees=45
  swing1Motion=Free
  swing2LimitDegrees=45
  swing2Motion=Free
  twistLimitDegrees=45
  twistMotion=Free
 }
 {
  bone1Name=dyn_cloth_01_r_fr
  bone2Name=root
  swing1LimitDegrees=45
  swing1Motion=Free
  swing2LimitDegrees=45
  swing2Motion=Free
  twistLimitDegrees=45
  twistMotion=Free
 }
]
