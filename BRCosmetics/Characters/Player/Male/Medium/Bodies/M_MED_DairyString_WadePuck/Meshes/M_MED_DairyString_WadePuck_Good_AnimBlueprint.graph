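# Animation Blueprint for the WadePuck (DairyString) player body cosmetic, adding secondary coat motion to the attached parent mesh's pose.
# Copied pose passes through Control Rig, head/neck counter-rotation and calf constraints.
# Cached result feeds RigidBody simulations for default, frontend idle or knee slide, then AnimDynamics tube/bow bones and additive corrective offsets.
# EventGraph derives state flags, head rotation, parameter-driven gravity, mannequin detection and gauntlet values from arm visibility.
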AnimBlueprint /BRCosmetics/Characters/Player/Male/Medium/Bodies/M_MED_DairyString_WadePuck/Meshes/M_MED_DairyString_WadePuck_Good_AnimBlueprint
target_skeleton=&/M_MED_DairyString_WadePuck_Skeleton
graphs [
 AnimGraph {nodes [
   Copy Pose From Mesh() {bUseAttachedParent=true}
   n1 = "Transform (Modify) Bone - Bone: dyn_cloth_01_r"!(
    ComponentPose=Local To Component(LocalPose=Use cached pose 'CRCache_Dynamics'())
    bAlphaBoolEnabled:bool=True
    Translation="0,0,0"
    Rotation="9.732,-1.797,-0.583"
    Scale="1,1,1"
   ) -> (Pose) {
    alphaInputType=Bool
    rotationMode=BMM_Additive
    rotationSpace=BCS_BoneSpace
    translationMode=BMM_Additive
    translationSpace=BCS_BoneSpace
    alphaBoolBlend {blendInTime=0.2, blendOutTime=0.2}
    boneToModify {boneName=dyn_cloth_01_r}
    rotation=<rot 9.732 -1.797 -0.583>
   }
   n2 = "Transform (Modify) Bone - Bone: dyn_cloth_01_r_bck"!(ComponentPose=@n1, bAlphaBoolEnabled:bool=True, Translation="0,0,0", Rotation="13.886,4.745,1.095", Scale="1,1,1") -> (Pose) {
    alphaInputType=Bool
    rotationMode=BMM_Additive
    rotationSpace=BCS_BoneSpace
    translationMode=BMM_Additive
    translationSpace=BCS_BoneSpace
    alphaBoolBlend {blendInTime=0.2, blendOutTime=0.2}
    boneToModify {boneName=dyn_cloth_01_r_bck}
    rotation=<rot 13.886 4.745 1.095>
   }
   n3 = "Transform (Modify) Bone - Bone: dyn_cloth_01_r_bck"(ComponentPose=@n2, bAlphaBoolEnabled=True, Translation="0,0,0", Rotation="36.371,-0.972,-0.377", Scale="1,1,1") {
    alphaInputType=Bool
    rotationMode=BMM_Additive
    rotationSpace=BCS_BoneSpace
    translationMode=BMM_Additive
    translationSpace=BCS_BoneSpace
    alphaBoolBlend {blendInTime=0.2, blendOutTime=0.2}
    boneToModify {boneName=dyn_cloth_01_r_bck}
    rotation=<rot 36.371 -0.972 -0.377>
   }
   n4 = "Transform (Modify) Bone - Bone: dyn_cloth_01_bck"!(ComponentPose=@n3, bAlphaBoolEnabled:bool=True, Translation="0,0,0", Rotation="6.031,-0.216,0.091", Scale="0,1,1") -> (Pose) {
    alphaInputType=Bool
    rotationMode=BMM_Additive
    rotationSpace=BCS_BoneSpace
    translationMode=BMM_Additive
    translationSpace=BCS_BoneSpace
    alphaBoolBlend {blendInTime=0.2, blendOutTime=0.2}
    boneToModify {boneName=dyn_cloth_01_bck}
    rotation=<rot 6.031 -0.216 0.091>
    scale {x=0}
   }
   n5 = "Transform (Modify) Bone - Bone: dyn_cloth_01_l_bck"!(ComponentPose=@n4, bAlphaBoolEnabled:bool=True, Translation="0,0,0", Rotation="16.012,0.015,-0.001", Scale="1,1,1") -> (Pose) {
    alphaInputType=Bool
    rotationMode=BMM_Additive
    rotationSpace=BCS_BoneSpace
    translationMode=BMM_Additive
    translationSpace=BCS_BoneSpace
    alphaBoolBlend {blendInTime=0.2, blendOutTime=0.2}
    boneToModify {boneName=dyn_cloth_01_l_bck}
    rotation=<rot 16.012 0.015 -0.001>
   }
   Output Pose(Result=Component To Local(ComponentPose=@n5))
   n6 = Control Rig!(
    Source=Copy Pose From Mesh() {bUseAttachedParent=true}
    Alpha
    isEmoting:bool=Get bIsPlayingEmote()
    isManequin:bool="Get isMannequin "() -> ("isMannequin ":bool)
    isKneeSlide:bool=Get bIsKneeSliding()
    isDBNO:bool=Get bIsDBNO()
    isCrouching:bool=Get bIsCrouching()
    guanteletEquipped:double=Get IsWearingGauntlet() -> (IsWearingGauntlet:double)
    rigidbodyenable:bool=Get dynEnable() -> (dynEnable:bool)
   ) -> (Pose) {controlRigAssetReference {blueprintRigClass=&/M_MED_DairyString_WadePuck_CtrlRig.M_MED_DairyString_WadePuck_CtrlRig_C}}
   n7 = Local To Component(LocalPose=@n6)
   n8 = Get head rot() -> (head rot:Rotator)
   n9 = "Transform (Modify) Bone - Bone: base_cloth_r_fr"!(
    ComponentPose="Transform (Modify) Bone - Bone: neck_02"(
     ComponentPose="Transform (Modify) Bone - Bone: head"(ComponentPose=@n7, Alpha=1, Translation="0,0,0", Rotation=@n8, Scale="1,1,1") {rotationMode=BMM_Additive, rotationSpace=BCS_BoneSpace, boneToModify {boneName=head}}
     Alpha=1
     Translation="0,0,0"
     Rotation=@n8
     Scale="1,1,1"
    ) {rotationMode=BMM_Additive, rotationSpace=BCS_BoneSpace, boneToModify {boneName=neck_02}}
    bAlphaBoolEnabled:bool=True
    Translation="0.000,-0.000,5.634"
    Rotation="-0.000,-0.000,33.236"
    Scale="1,1,1"
   ) -> (Pose) {
    alphaInputType=Bool
    rotationMode=BMM_Additive
    rotationSpace=BCS_BoneSpace
    translationMode=BMM_Additive
    translationSpace=BCS_BoneSpace
    alphaBoolBlend {blendInTime=0.2, blendOutTime=0.2}
    boneToModify {boneName=base_cloth_r_fr}
    rotation=<rot 0 0 33.236>
    translation {z=5.634}
   }
   n10 = "Constraint - dyn_cloth_02_l_bck (R)"!(ComponentPose=@n9, bAlphaBoolEnabled:bool=True, ConstraintWeights_0=1) -> (Pose) {
    alphaInputType=Bool
    boneToModify {boneName=dyn_cloth_02_l_bck}
    constraintSetup [{transformType=Rotation, targetBone {boneName=calf_l}}]
    constraintWeights [1]
   }
   n11 = "Constraint - dyn_cloth_02_r_bck (R)"!(ComponentPose=@n10, bAlphaBoolEnabled:bool=True, ConstraintWeights_0=1) -> (Pose) {
    alphaInputType=Bool
    boneToModify {boneName=dyn_cloth_02_r_bck}
    constraintSetup [{transformType=Rotation, targetBone {boneName=calf_r}}]
    constraintWeights [1]
   }
   n12 = "Constraint - dyn_cloth_02_bck (RR)"!(ComponentPose=@n11, bAlphaBoolEnabled:bool=True, ConstraintWeights_0=0.5, ConstraintWeights_1:float=0.5) -> (Pose) {
    alphaInputType=Bool
    boneToModify {boneName=dyn_cloth_02_bck}
    constraintSetup [{transformType=Rotation, targetBone {boneName=calf_r}}, {transformType=Rotation, targetBone {boneName=calf_l}}]
    constraintWeights [1,1]
   }
   Save cached pose 'CRCache_Input'(Pose=Component To Local(ComponentPose=@n12)) {cacheName=CRCache_Input}
   n13 = Anim Dynamics - Bone: dyn_tube_01_m(ComponentPose=Local To Component(LocalPose=Use cached pose 'CRCache_CoatDyn'()), Alpha=1, ExternalForce="0,0,0") {
    angularDampingOverride=1
    angularSpringConstant=4000
    bAngularSpring=true
    bOverrideAngularDamping=true
    bUseGravityOverride=true
    gravityScale=0.35
    boundBone {boneName=dyn_tube_01_m}
    componentAppliedLinearAccClamp=<300 300 300>
    componentLinearAccScale=<0.3 0.3 0.3>
    componentLinearVelScale=<0.3 0.3 0.3>
    gravityOverride {z=-980}
    physicsBodyDefinitions [{
      boundBone {boneName=dyn_tube_01_m}
      boxExtents=<20 30 30>
      constraintSetup {angularLimitsMax=<0 0 15>, angularLimitsMin=<0 0 -15>, angularTarget=<1 0 0>}
      localJointOffset=<3 0 0>
     }]
   }
   n14 = Anim Dynamics - Bone: dyn_tube_02_m(ComponentPose=@n13, Alpha=1, ExternalForce="0,0,0") {
    angularDampingOverride=1
    angularSpringConstant=8000
    bAngularSpring=true
    bOverrideAngularDamping=true
    bUseGravityOverride=true
    gravityScale=0.35
    boundBone {boneName=dyn_tube_02_m}
    componentAppliedLinearAccClamp=<300 300 300>
    componentLinearAccScale=<0.3 0.3 0.3>
    componentLinearVelScale=<0.3 0.3 0.3>
    gravityOverride {z=-980}
    physicsBodyDefinitions [{
      boundBone {boneName=dyn_tube_02_m}
      boxExtents=<30 30 30>
      constraintSetup {angularLimitsMax=<0 0 15>, angularLimitsMin=<0 0 -15>, angularTarget=<1 0 0>}
      localJointOffset=<4 0 0>
     }]
   }
   n15 = Anim Dynamics - Bone: base_tube_03_m(ComponentPose=@n14, Alpha=1, ExternalForce="0,0,0") {
    angularDampingOverride=1
    angularSpringConstant=10000
    bAngularSpring=true
    bOverrideAngularDamping=true
    bUseGravityOverride=true
    gravityScale=0.35
    boundBone {boneName=base_tube_03_m}
    componentAppliedLinearAccClamp=<300 300 300>
    componentLinearAccScale=<0.3 0.3 0.3>
    componentLinearVelScale=<0.3 0.3 0.3>
    gravityOverride {z=-980}
    physicsBodyDefinitions [{
      boundBone {boneName=base_tube_03_m}
      boxExtents=<30 30 30>
      constraintSetup {angularLimitsMax=<0 0 30>, angularLimitsMin=<0 0 -30>, angularTarget=<1 0 0>}
      localJointOffset=<2 0 0>
     }]
   }
   n16 = Anim Dynamics - Bone: dyn_bowWing_l(ComponentPose=@n15, Alpha=1, ExternalForce="0,0,0") {
    angularDampingOverride=1
    angularSpringConstant=10000
    bAngularSpring=true
    bOverrideAngularDamping=true
    bUseGravityOverride=true
    gravityScale=0.35
    boundBone {boneName=dyn_bowWing_l}
    componentLinearAccScale=<1 1 1>
    componentLinearVelScale=<1 1 1>
    gravityOverride {z=-980}
    physicsBodyDefinitions [{
      boundBone {boneName=dyn_bowWing_l}
      boxExtents=<20 30 30>
      constraintSetup {angularLimitsMax=<15 0 0>, angularLimitsMin=<-15 0 0>, angularTarget=<1 0 0>}
      localJointOffset=<4 0 0>
     }]
   }
   n17 = Anim Dynamics - Bone: dyn_bowWing_r(ComponentPose=@n16, Alpha=1, ExternalForce="0,0,0") {
    angularDampingOverride=1
    angularSpringConstant=10000
    bAngularSpring=true
    bOverrideAngularDamping=true
    bUseGravityOverride=true
    gravityScale=0.35
    boundBone {boneName=dyn_bowWing_r}
    componentLinearAccScale=<1 1 1>
    componentLinearVelScale=<1 1 1>
    gravityOverride {z=-980}
    physicsBodyDefinitions [{
      boundBone {boneName=dyn_bowWing_r}
      boxExtents=<20 30 30>
      constraintSetup {angularLimitsMax=<15 0 0>, angularLimitsMin=<-15 0 0>, angularTarget=<1 0 0>}
      localJointOffset=<4 0 0>
     }]
   }
   n18 = "Transform (Modify) Bone - Bone: dyn_cloth_01_l_fr"(ComponentPose=@n17, Alpha=1, Translation="0,0,0", Rotation="-17.148,0.000,-0.000", Scale="1,1,1") {
    rotationMode=BMM_Additive
    rotationSpace=BCS_BoneSpace
    translationMode=BMM_Additive
    translationSpace=BCS_BoneSpace
    boneToModify {boneName=dyn_cloth_01_l_fr}
    rotation=<rot -17.148 0 0>
   }
   n19 = "Transform (Modify) Bone - Bone: dyn_cloth_02_bck"!(ComponentPose=@n18, bAlphaBoolEnabled:bool=True, Translation="0,0,0", Rotation="25,0,0", Scale="1,1,1") -> (Pose) {
    alphaInputType=Bool
    rotationMode=BMM_Additive
    rotationSpace=BCS_BoneSpace
    translationMode=BMM_Additive
    translationSpace=BCS_BoneSpace
    alphaBoolBlend {blendInTime=0.2, blendOutTime=0.2}
    boneToModify {boneName=dyn_cloth_02_bck}
    rotation {pitch=25}
   }
   n20 = "Transform (Modify) Bone - Bone: dyn_cloth_02_l_bck"!(ComponentPose=@n19, bAlphaBoolEnabled:bool=True, Translation="0,0,0", Rotation="25,0,0", Scale="1,1,1") -> (Pose) {
    alphaInputType=Bool
    rotationMode=BMM_Additive
    rotationSpace=BCS_BoneSpace
    translationMode=BMM_Additive
    translationSpace=BCS_BoneSpace
    alphaBoolBlend {blendInTime=0.2, blendOutTime=0.2}
    boneToModify {boneName=dyn_cloth_02_l_bck}
    rotation {pitch=25}
   }
   n21 = "Transform (Modify) Bone - Bone: dyn_backPush"!(ComponentPose=@n20, bAlphaBoolEnabled:bool=True, Translation="15.561,17.803,0.006", Rotation="0,-25,0", Scale="0.315,0.315,0.315") -> (Pose) {
    alphaInputType=Bool
    rotationMode=BMM_Additive
    rotationSpace=BCS_BoneSpace
    scaleMode=BMM_Additive
    scaleSpace=BCS_BoneSpace
    translationMode=BMM_Additive
    translationSpace=BCS_BoneSpace
    alphaBoolBlend {blendInTime=0.2, blendOutTime=0.2}
    boneToModify {boneName=dyn_backPush}
    rotation {yaw=-25}
    scale=<0.3154 0.3154 0.3154>
    translation=<15.561 17.803 0.006>
   }
   Save cached pose 'CRCache_Dynamics'(Pose=Component To Local(ComponentPose=@n21)) {cacheName=CRCache_Dynamics}
   n22 = RigidBody!(
    ComponentPose=Local To Component(LocalPose=Use cached pose 'CRCache_Input'())
    Alpha
    OverrideWorldGravity:Vector="0,0,0"
    ExternalForce
    SimSpaceSettings:Sim Space Settings Structure=(WorldAlpha=1,VelocityScaleZ=1,DampingAlpha=1,MaxLinearVelocity=1,MaxAngularVelocity=1000,MaxLinearAcceleration=3,MaxAngularAcceleration=2000,ExternalLinearDragV=(X=0.2,Y=0.2,Z=0),ExternalLinearVelocity=(X=0,Y=0,Z=0),ExternalAngularVelocity=(X=0,Y=0,Z=0))
   ) -> (Pose) {
    bOverrideWorldGravity=true
    overridePhysicsAsset=&/M_MED_DairyString_WadePuck_Physics
    simulationSpace=BaseBoneSpace
    baseBoneRef {boneName=pelvis}
    componentAppliedLinearAccClamp=<300 300 300>
    simSpaceSettings {
     maxAngularAcceleration=2000
     maxAngularVelocity=1000
     maxLinearAcceleration=3
     maxLinearVelocity=1
     worldAlpha=1
     externalLinearDragV=<0.2 0.2>
    }
   }
   n23 = Component To Local(ComponentPose=@n22)
   n24 = RigidBody(
    ComponentPose=Local To Component(LocalPose=Use cached pose 'CRCache_Input'())
    OverrideWorldGravity="0,0,0"
    SimSpaceSettings=(WorldAlpha=1,VelocityScaleZ=1,DampingAlpha=1,MaxLinearVelocity=1,MaxAngularVelocity=1000,MaxLinearAcceleration=3,MaxAngularAcceleration=2000,ExternalLinearDragV=(X=0.2,Y=0.2,Z=0),ExternalLinearVelocity=(X=0,Y=0,Z=0),ExternalAngularVelocity=(X=0,Y=0,Z=0))
   ) {
    bOverrideWorldGravity=true
    overridePhysicsAsset=&/M_MED_DairyString_WadePuck_FE_Physics
    simulationSpace=BaseBoneSpace
    baseBoneRef {boneName=pelvis}
    componentAppliedLinearAccClamp=<300 300 300>
    simSpaceSettings {
     maxAngularAcceleration=2000
     maxAngularVelocity=1000
     maxLinearAcceleration=3
     maxLinearVelocity=1
     worldAlpha=1
     externalLinearDragV=<0.2 0.2>
    }
   }
   n25 = Component To Local(ComponentPose=@n24)
   n26 = Get isFE_idle() -> (isFE_idle:bool)
   n27 = "Transform (Modify) Bone - Bone: dyn_cloth_01_bck"!(ComponentPose=Local To Component(LocalPose=Use cached pose 'CRCache_Input'()), Alpha:float=1, Rotation="0.136,10.849,-0.725") -> (Pose) {
    rotationMode=BMM_Additive
    rotationSpace=BCS_BoneSpace
    translationMode=BMM_Additive
    translationSpace=BCS_BoneSpace
    boneToModify {boneName=dyn_cloth_01_bck}
    rotation=<rot 0.136 10.849 -0.725>
    translation=<-6.4595 -1.7431 4.8523>
   }
   n28 = "Transform (Modify) Bone - Bone: dyn_cloth_01_l_bck"!(ComponentPose=@n27, Alpha:float=1, Rotation="59.716,33.558,29.839") -> (Pose) {
    rotationMode=BMM_Additive
    rotationSpace=BCS_BoneSpace
    translationMode=BMM_Additive
    translationSpace=BCS_BoneSpace
    boneToModify {boneName=dyn_cloth_01_l_bck}
    rotation=<rot 59.716 33.558 29.839>
    translation=<-1.4426 -0.9569 -2.9644>
   }
   n29 = "Transform (Modify) Bone - Bone: dyn_cloth_01_r_bck"!(ComponentPose=@n28, Alpha:float=1, Rotation="-8.054,-1.357,-82.840") -> (Pose) {
    rotationMode=BMM_Additive
    rotationSpace=BCS_BoneSpace
    translationMode=BMM_Additive
    translationSpace=BCS_BoneSpace
    boneToModify {boneName=dyn_cloth_01_r_bck}
    rotation=<rot -8.054 -1.357 -82.84>
    translation=<-0.8089 0.1267 11.5198>
   }
   n30 = RigidBody(
    ComponentPose=@n29
    OverrideWorldGravity="0,0,0"
    SimSpaceSettings=(WorldAlpha=1,VelocityScaleZ=1,DampingAlpha=1,MaxLinearVelocity=1,MaxAngularVelocity=1000,MaxLinearAcceleration=3,MaxAngularAcceleration=2000,ExternalLinearDragV=(X=0.2,Y=0.2,Z=0),ExternalLinearVelocity=(X=0,Y=0,Z=0),ExternalAngularVelocity=(X=0,Y=0,Z=0))
   ) {
    bOverrideWorldGravity=true
    overridePhysicsAsset=&/M_MED_DairyString_WadePuck_KneeSliding_Physics
    simulationSpace=BaseBoneSpace
    baseBoneRef {boneName=pelvis}
    componentAppliedLinearAccClamp=<300 300 300>
    simSpaceSettings {
     maxAngularAcceleration=2000
     maxAngularVelocity=1000
     maxLinearAcceleration=3
     maxLinearVelocity=1
     worldAlpha=1
     externalLinearDragV=<0.2 0.2>
    }
   }
   n31 = "Constraint - dyn_cloth_02_l_bck (R)"(ComponentPose=@n30, bAlphaBoolEnabled=True, ConstraintWeights_0=1) {
    alphaInputType=Bool
    boneToModify {boneName=dyn_cloth_02_l_bck}
    constraintSetup [{transformType=Rotation, targetBone {boneName=calf_l}}]
    constraintWeights [1]
   }
   n32 = "Constraint - dyn_cloth_02_r_bck (R)"(ComponentPose=@n31, bAlphaBoolEnabled=True, ConstraintWeights_0=1) {
    alphaInputType=Bool
    boneToModify {boneName=dyn_cloth_02_r_bck}
    constraintSetup [{transformType=Rotation, targetBone {boneName=calf_r}}]
    constraintWeights [1]
   }
   n33 = "Constraint - dyn_cloth_02_bck (RR)"(ComponentPose=@n32, bAlphaBoolEnabled=True, ConstraintWeights_0=0.5, ConstraintWeights_1=0.5) {
    alphaInputType=Bool
    boneToModify {boneName=dyn_cloth_02_bck}
    constraintSetup [{transformType=Rotation, targetBone {boneName=calf_r}}, {transformType=Rotation, targetBone {boneName=calf_l}}]
    constraintWeights [1,1]
   }
   Save cached pose 'CRCache_CoatDyn'(Pose=Blend Poses by bool(
     BlendPose_0=Component To Local(ComponentPose=@n33)
     BlendPose_1=Blend Poses by bool(BlendPose_0=@n25, BlendPose_1=@n23, bActiveValue=@n26)
     BlendTime_0=1
     bActiveValue=Get bIsKneeSliding()
    )) {cacheName=CRCache_CoatDyn}
  ]}
 EventGraph {nodes [
   Event BlueprintUpdateAnimation()
   > n1 = Sequence!(execute) -> (then_0, then_1, then_2:Exec, then_3:Exec)
   n2 = Sequence(execute=@n1.then_2)
   n3 = Get bRigidBodyEnabled()
   Set ASW(execute=@n2.then_2, ASW:bool=AND Boolean(A=Get bIsSkyDiving(), B=NOT Boolean(A=@n3))) -> (Output_Get:bool)
   n4 = Get bRigidBodyEnabled()
   > Set ASWCrouch(ASWCrouch:bool=AND Boolean(A=Get bIsCrouching(), B=NOT Boolean(A=@n4))) -> (Output_Get:bool)
   n5 = Get bIsPlayingEmote()
   > Set isFE_idle(isFE_idle:bool=AND Boolean(A=Get bIsInFrontEnd(), B=NOT Boolean(A=@n5))) -> (Output_Get:bool)
   n6 = Get bAnimDynamicsEnabled() -> (bAnimDynamicsEnabled:bool)
   n7 = Get bRigidBodyEnabled()
   > Set dynEnable(dynEnable:bool=AND Boolean(A=NOT Boolean(A=@n7), B=NOT Boolean(A=@n6))) -> (Output_Get:bool)
   n8 = GetOwnerSkeletalMeshComponent()
   n9 = BreakRotator(InRot=BreakTransform(InTransform=GetDeltaTransformFromRefPose(self=@n8, BoneName=head, BaseName=neck_02)).Rotation)
   n10 = Get Z percentage() -> (Z percentage:float)
   n11 = Get X percentage() -> (X percentage:float)
   n12 = Get Y percentage() -> (Y percentage:float)
   Set head rot(
    execute=@n2.then_0
    head rot:Rotator=MakeRotator(
     Roll=float * float!(A=@n9, B=-1, C:double=@n11) -> (ReturnValue)
     Pitch=float * float(A=@n9.Pitch, B=-1, C=@n12)
     Yaw=float * float(A=@n9.Yaw, B=-1, C=@n10)
    )
   ) -> (Output_Get:Rotator)
   n13 = Branch(execute=@n2.then_1, Condition=Get bIsCrouching())
   n14 = BreakTransform(InTransform=GetDeltaTransformFromRefPose(self=GetOwnerSkeletalMeshComponent(), BoneName=pelvis, BaseName=root))
   > Set CrouchIdle(CrouchIdle:double=MapRangeClamped(Value=BreakVector(InVec=@n14), InRangeA=-30, InRangeB=-48)) -> (Output_Get:double)
   Set CrouchIdle(execute=@n13.else, CrouchIdle=0)
   n15 = "Gravity Override State Function (Parameter Asset) (Nativized)"(
    execute=@n1.then_0
    CharacterPartAnimInstance=Self-Reference()
    GameTimeInSeconds=GetGameTimeInSeconds()
    Parameters=&/M_MED_DairyString_WadePuck_Parameters
   )
   > Set GravityOverride(GravityOverride=@n15)
   > Set RBAN Settings Out(RBAN Settings Out=@n15.OutRigidBodyAnimNodeParameters)
   n16 = GetOwningActor()
   Cast To FortPlayerMannequin!(execute=@n1.then_1, Object=@n16) -> (then, CastFailed, AsFort Player Mannequin:Fort Player Mannequin Object Reference) {is=title K2Node_DynamicCast}
   > n17 = For Each Loop(Array:Array of Skeletal Mesh Component Object References=Get Components By Class!(self:Actor Object Reference=@n16, ComponentClass:Actor Component Class Reference=/Script/Engine.SkeletalMeshComponent) -> (ReturnValue:Array of Skeletal Mesh Component Object References)) -> (Array Element:Skeletal Mesh Component Object Reference)
   > n18 = Branch(Condition="Equal (Object)"(
     A=Get SkeletalMesh!(self:Skinned Mesh Component Object Reference=@n17) -> (SkeletalMesh:Skeletal Mesh Object Reference)
     B=&/M_MED_DairyString_WadePuck
    ))
   > "Set isMannequin "("isMannequin ":bool=true) -> (Output_Get:bool)
   "Set isMannequin "(execute=@n18.else, "isMannequin "=false)
   Event OnEndFalling()
   > Set jumpfalling(jumpfalling:bool=false) -> (Output_Get:bool)
   Event OnBeginJumping()
   > Set jumpfalling(jumpfalling=true)
   n19 = Event OnRightArmHidden() {is=title K2Node_Event}
   Set IsWearingGauntlet(execute=[@n19.then, @n21.then], IsWearingGauntlet:double=0.2) -> (Output_Get:double)
   n20 = Event OnRightArmVisible() {is=title K2Node_Event}
   Set IsWearingGauntlet(execute=[@n20.then, @n22.then], IsWearingGauntlet=1)
   n21 = Event OnLeftArmHidden() {is=title K2Node_Event}
   n22 = Event OnLeftArmVisible() {is=title K2Node_Event}
   TryGetPawnOwner()
  ]}
]
variables [
 ASW {type=boolean}
 ASWCrouch {type=boolean}
 CrouchIdle {type=number}
 dynEnable {type=boolean}
 GravityOverride {type=Vector}
 head rot {type=Rotator}
 isFE_idle {type=boolean}
 "isMannequin " {type=boolean}
 IsWearingGauntlet {type=number}
 jumpfalling {type=boolean}
 RBAN Settings Out {type=FortRigidBodyAnimNodeParameters}
 X percentage {type=number, default=0.9}
 Y percentage {type=number, default=0.9}
 Z percentage {type=number, default=0.9}
]
