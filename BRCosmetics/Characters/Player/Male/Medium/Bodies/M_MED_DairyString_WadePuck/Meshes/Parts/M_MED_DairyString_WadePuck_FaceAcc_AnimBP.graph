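# Post-attach AnimBlueprint for the WadePuck face accessory mesh, adding secondary physics motion to its hair bones.
# It copies the attached parent mesh's pose, then runs twelve Anim Dynamics nodes in series over left and right front, middle and back hair bones.
# Each bone is a separate body with identical angular spring, limits, reduced gravity and damped component-motion influence.
# The event graph has no logic, and the head rotation and percentage variables are unused.
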AnimBlueprint /BRCosmetics/Characters/Player/Male/Medium/Bodies/M_MED_DairyString_WadePuck/Meshes/Parts/M_MED_DairyString_WadePuck_FaceAcc_AnimBP
target_skeleton=&/M_MED_DairyString_WadePuck_FaceAcc_Skeleton
graphs [
 AnimGraph {nodes [
   n1 = Anim Dynamics - Bone: dyn_hairFr_01_l(ComponentPose=Local To Component(LocalPose=Copy Pose From Mesh() {bUseAttachedParent=true}), Alpha=1, ExternalForce="0,0,0") {
    angularDampingOverride=1
    angularSpringConstant=20000
    bAngularSpring=true
    bOverrideAngularDamping=true
    bUseGravityOverride=true
    gravityScale=0.35
    boundBone {boneName=dyn_hairFr_01_l}
    componentAppliedLinearAccClamp=<2000 2000 2000>
    componentLinearAccScale=<0.5 0.5 0.5>
    componentLinearVelScale=<0.5 0.5 0.5>
    gravityOverride {z=-980}
    physicsBodyDefinitions [{
      boundBone {boneName=dyn_hairFr_01_l}
      boxExtents=<30 30 30>
      constraintSetup {angularLimitsMax=<0 15 15>, angularLimitsMin=<0 -40 -15>, angularTarget=<1 0 0>}
      localJointOffset=<2 0 0>
     }]
    simSpaceSettings {maxAngularAcceleration=20000, maxAngularVelocity=1000, simSpaceAngularAlpha=1}
   }
   n2 = Anim Dynamics - Bone: dyn_hairFr_02_l(ComponentPose=@n1, Alpha=1, ExternalForce="0,0,0") {
    angularDampingOverride=1
    angularSpringConstant=20000
    bAngularSpring=true
    bOverrideAngularDamping=true
    bUseGravityOverride=true
    gravityScale=0.35
    boundBone {boneName=dyn_hairFr_02_l}
    componentAppliedLinearAccClamp=<2000 2000 2000>
    componentLinearAccScale=<0.5 0.5 0.5>
    componentLinearVelScale=<0.5 0.5 0.5>
    gravityOverride {z=-980}
    physicsBodyDefinitions [{
      boundBone {boneName=dyn_hairFr_02_l}
      boxExtents=<30 30 30>
      constraintSetup {angularLimitsMax=<0 15 15>, angularLimitsMin=<0 -40 -15>, angularTarget=<1 0 0>}
      localJointOffset=<2 0 0>
     }]
    simSpaceSettings {maxAngularAcceleration=20000, maxAngularVelocity=1000, simSpaceAngularAlpha=1}
   }
   n3 = Anim Dynamics - Bone: dyn_hairMid_01_l(ComponentPose=@n2, Alpha=1, ExternalForce="0,0,0") {
    angularDampingOverride=1
    angularSpringConstant=20000
    bAngularSpring=true
    bOverrideAngularDamping=true
    bUseGravityOverride=true
    gravityScale=0.35
    boundBone {boneName=dyn_hairMid_01_l}
    componentAppliedLinearAccClamp=<2000 2000 2000>
    componentLinearAccScale=<0.5 0.5 0.5>
    componentLinearVelScale=<0.5 0.5 0.5>
    gravityOverride {z=-980}
    physicsBodyDefinitions [{
      boundBone {boneName=dyn_hairMid_01_l}
      boxExtents=<30 30 30>
      constraintSetup {angularLimitsMax=<0 15 15>, angularLimitsMin=<0 -40 -15>, angularTarget=<1 0 0>}
      localJointOffset=<2 0 0>
     }]
    simSpaceSettings {maxAngularAcceleration=20000, maxAngularVelocity=1000, simSpaceAngularAlpha=1}
   }
   n4 = Anim Dynamics - Bone: dyn_hairMid_02_l(ComponentPose=@n3, Alpha=1, ExternalForce="0,0,0") {
    angularDampingOverride=1
    angularSpringConstant=20000
    bAngularSpring=true
    bOverrideAngularDamping=true
    bUseGravityOverride=true
    gravityScale=0.35
    boundBone {boneName=dyn_hairMid_02_l}
    componentAppliedLinearAccClamp=<2000 2000 2000>
    componentLinearAccScale=<0.5 0.5 0.5>
    componentLinearVelScale=<0.5 0.5 0.5>
    gravityOverride {z=-980}
    physicsBodyDefinitions [{
      boundBone {boneName=dyn_hairMid_02_l}
      boxExtents=<30 30 30>
      constraintSetup {angularLimitsMax=<0 15 15>, angularLimitsMin=<0 -40 -15>, angularTarget=<1 0 0>}
      localJointOffset=<2 0 0>
     }]
    simSpaceSettings {maxAngularAcceleration=20000, maxAngularVelocity=1000, simSpaceAngularAlpha=1}
   }
   n5 = Anim Dynamics - Bone: dyn_hairBck_01_l(ComponentPose=@n4, Alpha=1, ExternalForce="0,0,0") {
    angularDampingOverride=1
    angularSpringConstant=20000
    bAngularSpring=true
    bOverrideAngularDamping=true
    bUseGravityOverride=true
    gravityScale=0.35
    boundBone {boneName=dyn_hairBck_01_l}
    componentAppliedLinearAccClamp=<2000 2000 2000>
    componentLinearAccScale=<0.5 0.5 0.5>
    componentLinearVelScale=<0.5 0.5 0.5>
    gravityOverride {z=-980}
    physicsBodyDefinitions [{
      boundBone {boneName=dyn_hairBck_01_l}
      boxExtents=<30 30 30>
      constraintSetup {angularLimitsMax=<0 15 15>, angularLimitsMin=<0 -40 -15>, angularTarget=<1 0 0>}
      localJointOffset=<2 0 0>
     }]
    simSpaceSettings {maxAngularAcceleration=20000, maxAngularVelocity=1000, simSpaceAngularAlpha=1}
   }
   n6 = Anim Dynamics - Bone: dyn_hairBck_02_l(ComponentPose=@n5, Alpha=1, ExternalForce="0,0,0") {
    angularDampingOverride=1
    angularSpringConstant=20000
    bAngularSpring=true
    bOverrideAngularDamping=true
    bUseGravityOverride=true
    gravityScale=0.35
    boundBone {boneName=dyn_hairBck_02_l}
    componentAppliedLinearAccClamp=<2000 2000 2000>
    componentLinearAccScale=<0.5 0.5 0.5>
    componentLinearVelScale=<0.5 0.5 0.5>
    gravityOverride {z=-980}
    physicsBodyDefinitions [{
      boundBone {boneName=dyn_hairBck_02_l}
      boxExtents=<30 30 30>
      constraintSetup {angularLimitsMax=<0 15 15>, angularLimitsMin=<0 -40 -15>, angularTarget=<1 0 0>}
      localJointOffset=<2 0 0>
     }]
    simSpaceSettings {maxAngularAcceleration=20000, maxAngularVelocity=1000, simSpaceAngularAlpha=1}
   }
   n7 = Anim Dynamics - Bone: dyn_hairFr_01_r(ComponentPose=@n6, Alpha=1, ExternalForce="0,0,0") {
    angularDampingOverride=1
    angularSpringConstant=20000
    bAngularSpring=true
    bOverrideAngularDamping=true
    bUseGravityOverride=true
    gravityScale=0.35
    boundBone {boneName=dyn_hairFr_01_r}
    componentAppliedLinearAccClamp=<2000 2000 2000>
    componentLinearAccScale=<0.5 0.5 0.5>
    componentLinearVelScale=<0.5 0.5 0.5>
    gravityOverride {z=-980}
    physicsBodyDefinitions [{
      boundBone {boneName=dyn_hairFr_01_r}
      boxExtents=<30 30 30>
      constraintSetup {angularLimitsMax=<0 15 15>, angularLimitsMin=<0 -40 -15>, angularTarget=<1 0 0>}
      localJointOffset=<2 0 0>
     }]
    simSpaceSettings {maxAngularAcceleration=20000, maxAngularVelocity=1000, simSpaceAngularAlpha=1}
   }
   n8 = Anim Dynamics - Bone: dyn_hairFr_02_r(ComponentPose=@n7, Alpha=1, ExternalForce="0,0,0") {
    angularDampingOverride=1
    angularSpringConstant=20000
    bAngularSpring=true
    bOverrideAngularDamping=true
    bUseGravityOverride=true
    gravityScale=0.35
    boundBone {boneName=dyn_hairFr_02_r}
    componentAppliedLinearAccClamp=<2000 2000 2000>
    componentLinearAccScale=<0.5 0.5 0.5>
    componentLinearVelScale=<0.5 0.5 0.5>
    gravityOverride {z=-980}
    physicsBodyDefinitions [{
      boundBone {boneName=dyn_hairFr_02_r}
      boxExtents=<30 30 30>
      constraintSetup {angularLimitsMax=<0 15 15>, angularLimitsMin=<0 -40 -15>, angularTarget=<1 0 0>}
      localJointOffset=<2 0 0>
     }]
    simSpaceSettings {maxAngularAcceleration=20000, maxAngularVelocity=1000, simSpaceAngularAlpha=1}
   }
   n9 = Anim Dynamics - Bone: dyn_hairMid_01_r(ComponentPose=@n8, Alpha=1, ExternalForce="0,0,0") {
    angularDampingOverride=1
    angularSpringConstant=20000
    bAngularSpring=true
    bOverrideAngularDamping=true
    bUseGravityOverride=true
    gravityScale=0.35
    boundBone {boneName=dyn_hairMid_01_r}
    componentAppliedLinearAccClamp=<2000 2000 2000>
    componentLinearAccScale=<0.5 0.5 0.5>
    componentLinearVelScale=<0.5 0.5 0.5>
    gravityOverride {z=-980}
    physicsBodyDefinitions [{
      boundBone {boneName=dyn_hairMid_01_r}
      boxExtents=<30 30 30>
      constraintSetup {angularLimitsMax=<0 15 15>, angularLimitsMin=<0 -40 -15>, angularTarget=<1 0 0>}
      localJointOffset=<2 0 0>
     }]
    simSpaceSettings {maxAngularAcceleration=20000, maxAngularVelocity=1000, simSpaceAngularAlpha=1}
   }
   n10 = Anim Dynamics - Bone: dyn_hairMid_02_r(ComponentPose=@n9, Alpha=1, ExternalForce="0,0,0") {
    angularDampingOverride=1
    angularSpringConstant=20000
    bAngularSpring=true
    bOverrideAngularDamping=true
    bUseGravityOverride=true
    gravityScale=0.35
    boundBone {boneName=dyn_hairMid_02_r}
    componentAppliedLinearAccClamp=<2000 2000 2000>
    componentLinearAccScale=<0.5 0.5 0.5>
    componentLinearVelScale=<0.5 0.5 0.5>
    gravityOverride {z=-980}
    physicsBodyDefinitions [{
      boundBone {boneName=dyn_hairMid_02_r}
      boxExtents=<30 30 30>
      constraintSetup {angularLimitsMax=<0 15 15>, angularLimitsMin=<0 -40 -15>, angularTarget=<1 0 0>}
      localJointOffset=<2 0 0>
     }]
    simSpaceSettings {maxAngularAcceleration=20000, maxAngularVelocity=1000, simSpaceAngularAlpha=1}
   }
   n11 = Anim Dynamics - Bone: dyn_hairBck_01_r(ComponentPose=@n10, Alpha=1, ExternalForce="0,0,0") {
    angularDampingOverride=1
    angularSpringConstant=20000
    bAngularSpring=true
    bOverrideAngularDamping=true
    bUseGravityOverride=true
    gravityScale=0.35
    boundBone {boneName=dyn_hairBck_01_r}
    componentAppliedLinearAccClamp=<2000 2000 2000>
    componentLinearAccScale=<0.5 0.5 0.5>
    componentLinearVelScale=<0.5 0.5 0.5>
    gravityOverride {z=-980}
    physicsBodyDefinitions [{
      boundBone {boneName=dyn_hairBck_01_r}
      boxExtents=<30 30 30>
      constraintSetup {angularLimitsMax=<0 15 15>, angularLimitsMin=<0 -40 -15>, angularTarget=<1 0 0>}
      localJointOffset=<2 0 0>
     }]
    simSpaceSettings {maxAngularAcceleration=20000, maxAngularVelocity=1000, simSpaceAngularAlpha=1}
   }
   n12 = Anim Dynamics - Bone: dyn_hairBck_02_r(ComponentPose=@n11, Alpha=1, ExternalForce="0,0,0") {
    angularDampingOverride=1
    angularSpringConstant=20000
    bAngularSpring=true
    bOverrideAngularDamping=true
    bUseGravityOverride=true
    gravityScale=0.35
    boundBone {boneName=dyn_hairBck_02_r}
    componentAppliedLinearAccClamp=<2000 2000 2000>
    componentLinearAccScale=<0.5 0.5 0.5>
    componentLinearVelScale=<0.5 0.5 0.5>
    gravityOverride {z=-980}
    physicsBodyDefinitions [{
      boundBone {boneName=dyn_hairBck_02_r}
      boxExtents=<30 30 30>
      constraintSetup {angularLimitsMax=<0 15 15>, angularLimitsMin=<0 -40 -15>, angularTarget=<1 0 0>}
      localJointOffset=<2 0 0>
     }]
    simSpaceSettings {maxAngularAcceleration=20000, maxAngularVelocity=1000, simSpaceAngularAlpha=1}
   }
   Output Pose(Result=Component To Local(ComponentPose=@n12)) {layerGroup=None}
  ]}
 EventGraph {nodes [Event BlueprintUpdateAnimation()]}
]
variables [
 head rot {type=Rotator}
 X percentage {type=number, default=0.9}
 Y percentage {type=number, default=0.9}
 Z percentage {type=number, default=0.9}
]
